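; Area-effect zone for one rune of the Cube, applying a gameplay effect to players and toggling vehicle zone state and fall-damage immunity on overlap.
; Server enables ambient rune effects, waiting for the Cube to leave if it is on this rune.
; RepNotifies drive rune placement, material, ambient particles and printing glow attached to the Cube's down face.
; Clients disable collision and play looping audio; construction scales visuals from capsule radius; replay scrubs restore particles.
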(event Custom|TrySpawnGlowEffect)

(event Custom|CubeNoLongerOnTop
  (Development|PrintString "Hello")
  (Variables|Default|SetEnableRuneAmbientEffect true))

(event Custom|SpawnVolumeEffect
  (Variables|Default|SetEnableRuneAmbientEffect true))

(event EventBeginPlay
  (bind _returnvalue (BlueprintContext|GetFortReplayContext))
  (Utilities|FlowControl|Sequence
    (:then_0
      (Networking|SwitchHasAuthority
        (:Authority
          (bind _ascube (Utilities|Casting|CastToCUBE (Actor|GetOwner)))
          (bind (_currentstep_t_16_ea4278f841ee4e73c6861e88cb99517b _currentstep_2parttransitiontime_12_209590664625dbbf3f6200a38992da89 _currentstep_downface_19_20869c6c4df1517848398693617bdfc2 _currentstep_runenumber_21_3c4aa7924813a36f10d2fa820589c6a3) (Utilities|Struct|BreakCubeStruct (Variables|/Game/Athena/Prototype/Blueprints/Cube/CUBE.CUBE_C|GetCurrentStep_T_16_EA4278F841EE4E73C6861E88CB99517B _ascube)))
          (if (Math|Byte|Equal(Byte) _currentstep_runenumber_21_3c4aa7924813a36f10d2fa820589c6a3 (Math|Byte|Byte+Byte (Variables|Default|GetRuneIndex) 1))
            (Default|BindEventtoCubeLeftPosition _ascube (delegate Custom|SpawnVolumeEffect))
            (else
              (Variables|Default|SetEnableRuneAmbientEffect true))))
        (:Remote
          (Collision|SetCollisionEnabled (Variables|Default|GetEffect%20Area) "NoCollision")
          (bind _returnvalue_2 (Audio|SpawnSoundAttached "/Game/Athena/Sounds/Events/Cube/Cube_AreaEffect_LeftBehind_Loop_Cue.Cube_AreaEffect_LeftBehind_Loop_Cue" (Variables|Default|GetEffect%20Area)))
          (Audio|Components|Audio|FadeIn _returnvalue_2 1e+01 1.0 0.0 "Linear" :!class /Script/Engine.AudioComponent))))
    (:then_1
      (Utilities|IsValid _returnvalue
        (:Is%20Valid
          (ReplayContext|BindEventtoOnReplayScrubComplete _returnvalue (delegate Custom|replayscrubbed)))))))

(event Custom|replayscrubbed
  (Components|Activation|Activate (Variables|Default|GetP_Cube_AoE_Volume) true)
  (Components|Activation|SetActive (Variables|Default|GetP_Cube_Rune_Volume) (Variables|Default|GetEnableRuneAmbientEffect) true))

(event Custom|DisableGlowEffect
  (Variables|Default|SetPrinting false))

(event OnComponentBeginOverlap(EffectArea)
  (params (OverlappedComponent :type Primitive%20Component%20Object%20Reference) (OtherActor :type Actor%20Object%20Reference) (OtherComp :type Primitive%20Component%20Object%20Reference) (OtherBodyIndex :type Integer) (bFromSweep :type Boolean) (SweepResult :type Hit%20Result%20Structure%20%28by%20ref%29))
  (bind _asfort_player_pawn (Utilities|Casting|CastToFortPlayerPawn OtherActor)
    (:then
      (Environment|ApplyGlobalEnvironmentGameplayEffecttoActor _asfort_player_pawn "/Game/Athena/Prototype/Blueprints/Cube/GE_Cube_Area_Effect.GE_Cube_Area_Effect_C"))
    (:CastFailed
      (bind _asfort_vehicle_interface (Utilities|Casting|CastToFortVehicleInterface OtherActor))
      (CallFunction|OnVehicleEnterZone (Vehicle|GetActor _asfort_vehicle_interface) 0.5)
      (Vehicle|SetIgnoreAllFallingDamage _asfort_vehicle_interface true))))

(event Collision|EventActorEndOverlap
  (params (OtherActor :type Actor%20Object%20Reference))
  (bind _asfort_player_pawn (Utilities|Casting|CastToFortPlayerPawn OtherActor)
    (:then
      (GameplayEffects|RemoveActiveGameplayEffectbySourceEffect (Ability|GetAbilitySystemComponent _asfort_player_pawn) "/Game/Athena/Prototype/Blueprints/Cube/GE_Cube_Area_Effect.GE_Cube_Area_Effect_C" 0 -1))
    (:CastFailed
      (bind _asfort_vehicle_interface (Utilities|Casting|CastToFortVehicleInterface OtherActor))
      (CallFunction|OnVehicleExitZone (Vehicle|GetActor _asfort_vehicle_interface))
      (Vehicle|SetIgnoreAllFallingDamage _asfort_vehicle_interface false))))

(fn OnRep_EnableRuneAmbientEffect
  (Networking|SwitchHasAuthority
    (:Remote
      (Components|Activation|SetActive (Variables|Default|GetP_Cube_Rune_Volume) (Variables|Default|GetEnableRuneAmbientEffect) true))))

(fn OnRep_Printing
  (bind _p_cube_printing (Variables|Default|GetP_Cube_Printing))
  (if (Variables|Default|GetPrinting)
    (bind _ascube (Utilities|Casting|CastToCUBE (Actor|GetOwner)))
    (bind (_currentstep_t_16_ea4278f841ee4e73c6861e88cb99517b _currentstep_2parttransitiontime_12_209590664625dbbf3f6200a38992da89 _currentstep_downface_19_20869c6c4df1517848398693617bdfc2 _currentstep_runenumber_21_3c4aa7924813a36f10d2fa820589c6a3) (Utilities|Struct|BreakCubeStruct (Variables|/Game/Athena/Prototype/Blueprints/Cube/CUBE.CUBE_C|GetCurrentStep_T_16_EA4278F841EE4E73C6861E88CB99517B _ascube)))
    (if (Math|Byte|Equal(Byte) _currentstep_runenumber_21_3c4aa7924813a36f10d2fa820589c6a3 (Math|Byte|Byte+Byte (Variables|Default|GetRuneIndex) 1))
      (Transformation|AttachComponentToComponent _p_cube_printing (Variables|/Game/Athena/Prototype/Blueprints/Cube/CUBE.CUBE_C|GetStaticMesh _ascube) (Utilities|String|StringToName (Utilities|String|ToString(Byte) _currentstep_downface_19_20869c6c4df1517848398693617bdfc2)) "SnapToTarget" "SnapToTarget" "SnapToTarget" true)
      (Transformation|SetRelativeRotation _p_cube_printing "180.000000, 0, 0" false false)
      (Components|Activation|SetActive _p_cube_printing true true)
      (Default|BindEventtoStopRuneGlow _ascube (EventDispatchers|CreateEvent :!event DisableGlowEffect)))
    (else
      (Components|Activation|SetActive (Variables|Default|GetP_Cube_Printing) false false))))

(fn OnRep_RuneIndex
  (bind _ascube (Utilities|Casting|CastToCUBE (Actor|GetOwner)))
  (Utilities|FlowControl|Sequence
    (:then_1
      (bind (_output_location _output_rotation) (Math|Transform|BreakTransform (Utilities|Array|Get(acopy) (Variables|/Game/Athena/Prototype/Blueprints/Cube/CUBE.CUBE_C|GetRuneLocations _ascube) (Math|Conversions|ToInteger(Byte) (Variables|Default|GetRuneIndex)))))
      (Transformation|SetWorldTransform (Variables|Default|GetRuneMesh) (Math|Transform|MakeTransform _output_location _output_rotation "0.520000,0.520000,0.520000") false false)
      (CallFunction|SetRuneMaterial))))

(fn SetRuneMaterial
  (Rendering|Material|SetMaterial (Variables|Default|GetRuneMesh) 0 (Utilities|Select :Index (Math|Conversions|ToInteger(Byte) (Variables|Default|GetRuneIndex)) :Option%200 "/Game/Athena/Prototype/Blueprints/Cube/Materials/M_DFHeight_Inst01.M_DFHeight_Inst01" :Option%201 "/Game/Athena/Prototype/Blueprints/Cube/Materials/M_DFHeight_Inst02.M_DFHeight_Inst02" :Option%202 "/Game/Athena/Prototype/Blueprints/Cube/Materials/M_DFHeight_Inst03.M_DFHeight_Inst03" :Option%203 "/Game/Athena/Prototype/Blueprints/Cube/Materials/M_DFHeight_Inst04.M_DFHeight_Inst04" :Option%204 "/Game/Athena/Prototype/Blueprints/Cube/Materials/M_DFHeight_Inst05.M_DFHeight_Inst05" :Option%205 "/Game/Athena/Prototype/Blueprints/Cube/Materials/M_DFHeight_Inst06.M_DFHeight_Inst06" :Option%206 "/Game/Athena/Prototype/Blueprints/Cube/Materials/M_DFHeight_Inst07.M_DFHeight_Inst07")))

(fn ConstructionScript
  (bind _effect_area (Variables|Default|GetEffect%20Area))
  (bind _capsuleradius (Variables|/Script/Engine.CapsuleComponent|GetCapsuleRadius _effect_area))
  (bind _returnvalue (/ _capsuleradius 1e+03))
  (bind _areasphere (Variables|Default|GetAreaSphere))
  (Transformation|SetWorldScale3D _areasphere (Math|Conversions|ToVector(Float) _returnvalue))
  (Rendering|Material|SetScalarParameterValueonMaterials _areasphere "SphereRadius" _capsuleradius)
  (Effects|Components|ParticleSystem|SetVectorParameter (Variables|Default|GetP_Cube_AoE_Volume) "SizeScale" (Math|Conversions|ToVector(Float) _returnvalue)))
