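; CubeOverlapActor is the CUBE's companion actor, handling its proximity effect, retaliation zap and cosmetic FX.
; Its sphere grants a gameplay effect on overlap, removed on exit or destruction.
; ZapPlayer telegraphs, retargets the closest player, then sphere-traces from the nearest non-bottom face socket, retrying downward on a miss, to damage and launch players, impulse vehicles or destroy blocking buildings.
; Timers, timelines and GameplayCues drive rune waves, glows and lightning.
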(event EventBeginPlay
  ;[1 1 1] Periodically spawn a wave of runes over the cube
  (Networking|CanExecuteCosmeticEvents
    (:True
      (Utilities|Time|SetTimerbyEvent (delegate Custom|CubeRuneWub) (Variables|Default|GetTimeBetweenRuneWubs) true))))

(event Custom|CubeRuneWub
  ;[1 1 1] Periodically spawn a wave of runes over the cube
  (if (not (Components|Timeline|IsPlaying (Variables|CubeOverlapActor|GetCubeRetaliationGlow)))
    (bind _ascube (Utilities|Casting|CastToCUBE (Actor|GetOwner)))
    (bind (_currentstep_t_16_ea4278f841ee4e73c6861e88cb99517b _currentstep_2parttransitiontime_12_209590664625dbbf3f6200a38992da89 _currentstep_downface_19_20869c6c4df1517848398693617bdfc2) (Utilities|Struct|BreakCubeStruct (Variables|/Game/Athena/Prototype/Blueprints/Cube/CUBE.CUBE_C|GetCurrentStep_T_16_EA4278F841EE4E73C6861E88CB99517B _ascube)))
    (bind _returnvalue (Effects|SpawnEmitterAttached :EmitterTemplate "/Game/Athena/Prototype/Blueprints/Cube/FX/P_Cube_Wave.P_Cube_Wave" :AttachToComponent (Variables|/Game/Athena/Prototype/Blueprints/Cube/CUBE.CUBE_C|GetStaticMesh _ascube) :AttachPointName (Utilities|String|StringToName (Utilities|String|ToString(Byte) _currentstep_downface_19_20869c6c4df1517848398693617bdfc2)) :Rotation "180.000000, 0, 0" :Scale "1.000000,1.000000,1.000000" :LocationType "SnapToTarget" :bAutoDestroy true :PoolingMethod "None" :bAutoActivate true))
    (Audio|SpawnSoundAttached "/Game/Athena/Sounds/Events/Cube/Cube_AmbPan_Cue.Cube_AmbPan_Cue" _returnvalue)))

(event Custom|SpawnGC
  (params (GCTarget :type Actor%20Object%20Reference) (Location :type Vector) (Normal :type Vector))
  (bind (_ascube _bsuccess) (Utilities|Casting|CastToCUBE GCTarget))
  ;[1 1 1] Replicated zap lightning FX
  (GameplayCue|ExecuteGameplayCue%28Burst%29Local GCTarget (Variables|Default|GetLightningImpactSCue) (Ability|GameplayCue|MakeGameplayCueParameters :NormalizedMagnitude (Math|Conversions|ToFloat(Boolean) _bsuccess) :RawMagnitude 0.0 :EffectContext "()" :AggregatedSourceTags "(GameplayTags=)" :AggregatedTargetTags "(GameplayTags=)" :Location (Math|Vector|MakeVectorNetQuantize10 (.x Location) (.y Location) (.z Location)) :Normal (Math|Vector|MakeVectorNetQuantizeNormal (.x Normal) (.y Normal) (.z Normal)) :SourceObject (Actor|GetOwner) :GameplayEffectLevel 1 :AbilityLevel 1 :bReplicateLocationWhenUsingMinimalRepProxy false)))

(event Custom|ZapVehicle
  (params (Vehicle :type Actor%20Object%20Reference) (Normal :type Vector) (Location :type Vector))
  (Class|CUBE|ImpulseVehicle (Utilities|Casting|CastToCUBE (Actor|GetOwner)) Vehicle Normal Location 1.0)
  (CallFunction|SpawnGC (Variables|Default|GetZapTarget) (Variables|Default|GetTraceHitLocation) (Variables|Default|GetTraceHitNormal)) ; GC creates lightning bolt from cube to vehicle
)

(event Custom|ZapPlayer
  (params (InstigatingPlayer :type Fort%20Player%20Pawn%20Object%20Reference))
  (Utilities|IsValid InstigatingPlayer
    (:Is%20Valid
      (Variables|Default|SetZapInstigator InstigatingPlayer)
      (Variables|Default|SetZapTarget InstigatingPlayer)
      (Variables|Default|SetTraceDown false)
      (CallFunction|CubeRetaliationTelegraph)
      (Variables|Default|SetCubeLookDirection "0, 0, 0") ; Reset
      (Utilities|FlowControl|Delay 2.0) ; Warmup telegraph time
      (bind _output_get (Variables|Default|SetOverlaps (Collision|GetOverlappingActors (Variables|Default|GetSphere) "/Game/Athena/PlayerPawn_Athena.PlayerPawn_Athena_C" :!class /Script/Engine.PrimitiveComponent)))
      (if (Utilities|IsArrayNonEmpty _output_get)
        (bind _output_get_2 (Variables|Default|SetZapTarget (Fortnite|GetClosestActorfromArray self (Variables|Default|GetOverlaps)))) ; Make closest player new target
      )
      (bind (_lookatdirection _socketlocation) (CallFunction|GetSocketClosesttoTarget _output_get_2)) ; Get nearest cube face, excluding down face
      (Variables|Default|SetCubeLookDirection _lookatdirection)
      (bind (_outhit _returnvalue) (Collision|SphereTraceByChannel _socketlocation (Math|Vector|vector-vector (+ _socketlocation (* (Math|Vector|Normalize (- (Transformation|GetActorLocation (Variables|Default|GetZapTarget)) _socketlocation)) (Variables|/Script/Engine.SphereComponent|GetSphereRadius (Variables|Default|GetSphere)))) (Math|Vector|MakeVector 0.0 0.0 (Math|Float|SelectFloat 1.5e+03 0.0 (Variables|Default|GetTraceDown)))) 14.0 "TraceTypeQuery1" true (Utilities|Array|MakeArray (Actor|GetOwner)) "ForDuration")) ; Trace toward target, only as far as sphere radius
      (if _returnvalue
        (Variables|Default|SetTraceHitLocation (.Location (Collision|BreakHitResult _outhit)))
        (Variables|Default|SetTraceHitNormal (.Normal (Collision|BreakHitResult _outhit)))
        (if (or (.bSuccess (Utilities|Casting|CastToFortPlayerPawn (.HitActor (Collision|BreakHitResult _outhit)))) (.bSuccess (Utilities|Casting|CastToFortVehicleInterface (.HitActor (Collision|BreakHitResult _outhit)))))
          (GameplayEffects|ApplyGameplayEffectToTarget :self (Ability|GetAbilitySystemComponent (Variables|Default|GetZapInstigator)) :GameplayEffectClass "/Game/Athena/Prototype/Blueprints/Cube/GE_CubeDamageResponse.GE_CubeDamageResponse_C" :Target (Ability|GetAbilitySystemComponent (Variables|Default|GetZapTarget)) :Level 0.0)
          (bind _ascube (Utilities|Casting|CastToCUBE (Actor|GetOwner)))
          (bind _asplayer_pawn_athena (Utilities|Casting|CastToPlayerPawn_Athena (Variables|Default|GetZapTarget)))
          (Class|FortPlayerPawnAthena|SetFallInstigator _asplayer_pawn_athena (Pawn|GetController (Variables|Default|GetZapInstigator)) "(GameplayTags=((TagName=\"Gameplay.Damage.Environment.Cube\")))")
          (if (Vehicle|IsinVehicle _asplayer_pawn_athena)
            (bind _returnvalue_2 (Vehicle|GetVehicleActor (Vehicle|GetVehicleInterface _asplayer_pawn_athena) :!class /Script/FortniteGame.FortVehicleInterface))
            (bind _output_get_3 (Variables|Default|SetZapTarget _returnvalue_2)) ; Change target to vehicle
            (CallFunction|ZapVehicle _output_get_3 (Math|Vector|Normalize (Variables|Default|GetDirectionToPlayer)) (Class|FortVehicleInterface|GetOwnerLocation (Vehicle|GetVehicleInterface _asplayer_pawn_athena)))
            (else
              (Class|CUBE|LaunchPlayerAway _ascube "0, 0, 0" (* (Math|Vector|Normalize (Variables|Default|GetDirectionToPlayer)) -1.0) _asplayer_pawn_athena 1.75 false)
              (CallFunction|SpawnGC (Variables|Default|GetZapTarget) (Transformation|GetActorLocation (Variables|Default|GetZapTarget))) ; GC creates lightning bolt from cube to player
            ))
          (else
            ;[1 1 1] Destroy blocking building actor, or hit static geometry
            (CallFunction|SpawnGC (Actor|GetOwner) (Variables|Default|GetTraceHitLocation) (Variables|Default|GetTraceHitNormal))
            (bind _asbuilding_actor (Utilities|Casting|CastToBuildingActor (.HitActor (Collision|BreakHitResult _outhit)))) ; Hit building?
            (if (Variables|Actor|GetbCanBeDamaged _asbuilding_actor)
              (Actor|DestroyActor _asbuilding_actor))))
        (elif (Variables|Default|GetTraceDown)
          ;[1 1 1] If trace missed, try one more time, aiming at ground
          (Variables|Default|SetCubeLookDirection "0, 0, 0")
          (else
            (Variables|Default|SetTraceDown true)
            (CallFunction|TraceDownward)))))) ; Stop here if damage has no instigator
)

(event Custom|TraceDownward
  (bind (_outhit _returnvalue) (Collision|SphereTraceByChannel (stale Custom|ZapPlayer _socketlocation) (Math|Vector|vector-vector (+ (stale Custom|ZapPlayer _socketlocation) (* (Math|Vector|Normalize (- (Transformation|GetActorLocation (Variables|Default|GetZapTarget)) (stale Custom|ZapPlayer _socketlocation))) (Variables|/Script/Engine.SphereComponent|GetSphereRadius (Variables|Default|GetSphere)))) (Math|Vector|MakeVector 0.0 0.0 (Math|Float|SelectFloat 1.5e+03 0.0 (Variables|Default|GetTraceDown)))) 14.0 "TraceTypeQuery1" true (Utilities|Array|MakeArray (Actor|GetOwner)) "ForDuration")) ; Trace toward target, only as far as sphere radius
  (if _returnvalue
    (Variables|Default|SetTraceHitLocation (.Location (Collision|BreakHitResult _outhit)))
    (Variables|Default|SetTraceHitNormal (.Normal (Collision|BreakHitResult _outhit)))
    (if (or (.bSuccess (Utilities|Casting|CastToFortPlayerPawn (.HitActor (Collision|BreakHitResult _outhit)))) (.bSuccess (Utilities|Casting|CastToFortVehicleInterface (.HitActor (Collision|BreakHitResult _outhit)))))
      (GameplayEffects|ApplyGameplayEffectToTarget :self (Ability|GetAbilitySystemComponent (Variables|Default|GetZapInstigator)) :GameplayEffectClass "/Game/Athena/Prototype/Blueprints/Cube/GE_CubeDamageResponse.GE_CubeDamageResponse_C" :Target (Ability|GetAbilitySystemComponent (Variables|Default|GetZapTarget)) :Level 0.0)
      (bind _ascube (Utilities|Casting|CastToCUBE (Actor|GetOwner)))
      (bind _asplayer_pawn_athena (Utilities|Casting|CastToPlayerPawn_Athena (Variables|Default|GetZapTarget)))
      (Class|FortPlayerPawnAthena|SetFallInstigator _asplayer_pawn_athena (Pawn|GetController (Variables|Default|GetZapInstigator)) "(GameplayTags=((TagName=\"Gameplay.Damage.Environment.Cube\")))")
      (if (Vehicle|IsinVehicle _asplayer_pawn_athena)
        (bind _returnvalue_2 (Vehicle|GetVehicleActor (Vehicle|GetVehicleInterface _asplayer_pawn_athena) :!class /Script/FortniteGame.FortVehicleInterface))
        (bind _output_get (Variables|Default|SetZapTarget _returnvalue_2)) ; Change target to vehicle
        (CallFunction|ZapVehicle _output_get (Math|Vector|Normalize (Variables|Default|GetDirectionToPlayer)) (Class|FortVehicleInterface|GetOwnerLocation (Vehicle|GetVehicleInterface _asplayer_pawn_athena)))
        (else
          (Class|CUBE|LaunchPlayerAway _ascube "0, 0, 0" (* (Math|Vector|Normalize (Variables|Default|GetDirectionToPlayer)) -1.0) _asplayer_pawn_athena 1.75 false)
          (CallFunction|SpawnGC (Variables|Default|GetZapTarget) (Transformation|GetActorLocation (Variables|Default|GetZapTarget))) ; GC creates lightning bolt from cube to player
        ))
      (else
        ;[1 1 1] Destroy blocking building actor, or hit static geometry
        (CallFunction|SpawnGC (Actor|GetOwner) (Variables|Default|GetTraceHitLocation) (Variables|Default|GetTraceHitNormal))
        (bind _asbuilding_actor (Utilities|Casting|CastToBuildingActor (.HitActor (Collision|BreakHitResult _outhit)))) ; Hit building?
        (if (Variables|Actor|GetbCanBeDamaged _asbuilding_actor)
          (Actor|DestroyActor _asbuilding_actor))))
    (elif (Variables|Default|GetTraceDown)
      ;[1 1 1] If trace missed, try one more time, aiming at ground
      (Variables|Default|SetCubeLookDirection "0, 0, 0")
      (else
        (Variables|Default|SetTraceDown true)
        (CallFunction|TraceDownward)))))

(event Custom|CubeRetaliationTelegraph
  ;[1 1 1] Cube zap telegraph animation
  (Networking|CanExecuteCosmeticEvents
    (:True
      (bind _ascube (Utilities|Casting|CastToCUBE (Actor|GetOwner)))
      (Fortnite|PlayLocalForceFeedbackatLocation "/Game/Athena/Prototype/FF_Cube_Mad.FF_Cube_Mad" (Transformation|GetWorldLocation (Variables|/Game/Athena/Prototype/Blueprints/Cube/CUBE.CUBE_C|GetStaticMesh _ascube)) 5e+03)
      (Audio|SpawnSoundAttached "/Game/Athena/Sounds/Events/Cube/Cube_AngryZap_Pre_Source_Cue.Cube_AngryZap_Pre_Source_Cue" (Variables|/Game/Athena/Prototype/Blueprints/Cube/CUBE.CUBE_C|GetAmbientAudio _ascube))
      (bind (_direction _glow _lookat _lookaround) (Timeline|CubeRetaliationGlow PlayFromStart)
        (:Update
          (Rendering|Material|SetScalarParameterValueonMaterials (Variables|/Game/Athena/Prototype/Blueprints/Cube/CUBE.CUBE_C|GetStaticMesh _ascube) "Zap" _glow)
          (Rendering|Material|SetVectorParameterValueonMaterials (Variables|/Game/Athena/Prototype/Blueprints/Cube/CUBE.CUBE_C|GetStaticMesh _ascube) "SphereOffset" (* (Math|Vector|Lerp(Vector) _lookaround (Variables|Default|GetCubeLookDirection) _lookat) 7e+02)))))))

(event Custom|CubeBounceFX
  (params (Location :type Vector) (Normal :type Vector) (Target :type Actor%20Object%20Reference))
  ;[1 1 1] Cube bounce player FX
  (Networking|CanExecuteCosmeticEvents
    (:True
      (GameplayCue|ExecuteGameplayCue%28Burst%29Local Target (Variables|Default|GetLightningImpactLCue) (Ability|GameplayCue|MakeGameplayCueParameters :NormalizedMagnitude 0.0 :RawMagnitude 0.0 :EffectContext "()" :AggregatedSourceTags "(GameplayTags=)" :AggregatedTargetTags "(GameplayTags=)" :Location (Math|Vector|MakeVectorNetQuantize10 (.x Location) (.y Location) (.z Location)) :Normal (Math|Vector|MakeVectorNetQuantizeNormal (.x Normal) (.y Normal) (.z Normal)) :Instigator self :GameplayEffectLevel 1 :AbilityLevel 1 :bReplicateLocationWhenUsingMinimalRepProxy false))
      (bind _ascube (Utilities|Casting|CastToCUBE (Actor|GetOwner)))
      (bind (_direction _flash _offset) (Timeline|CubeBounceFlash PlayFromStart)
        (:Update
          (Rendering|Material|SetScalarParameterValueonMaterials (Variables|/Game/Athena/Prototype/Blueprints/Cube/CUBE.CUBE_C|GetStaticMesh _ascube) "Zap2" _flash)
          (if (not (Components|Timeline|IsPlaying (Variables|CubeOverlapActor|GetCubeRetaliationGlow)))
            (Rendering|Material|SetVectorParameterValueonMaterials (Variables|/Game/Athena/Prototype/Blueprints/Cube/CUBE.CUBE_C|GetStaticMesh _ascube) "SphereOffset" (* (- Location (Transformation|GetWorldLocation (Variables|/Game/Athena/Prototype/Blueprints/Cube/CUBE.CUBE_C|GetStaticMesh _ascube))) _offset))))))))

(event OnComponentBeginOverlap(Sphere)
  (params (OverlappedComponent :type Primitive%20Component%20Object%20Reference) (OtherActor :type Actor%20Object%20Reference) (OtherComp :type Primitive%20Component%20Object%20Reference) (OtherBodyIndex :type Integer) (bFromSweep :type Boolean) (SweepResult :type Hit%20Result%20Structure%20%28by%20ref%29))
  (Environment|ApplyGlobalEnvironmentGameplayEffecttoActor OtherActor "/Game/Athena/Prototype/Blueprints/Cube/GE_CubePositiveEffect.GE_CubePositiveEffect_C" 0 "(GameplayTags=((TagName=\"Gameplay.Damage.Environment.Cube\")))"))

(event OnComponentEndOverlap(Sphere)
  (params (OverlappedComponent :type Primitive%20Component%20Object%20Reference) (OtherActor :type Actor%20Object%20Reference) (OtherComp :type Primitive%20Component%20Object%20Reference) (OtherBodyIndex :type Integer))
  (GameplayEffects|RemoveActiveGameplayEffectbySourceEffect (Ability|GetAbilitySystemComponent OtherActor) "/Game/Athena/Prototype/Blueprints/Cube/GE_CubePositiveEffect.GE_CubePositiveEffect_C" 0 -1))

(event EventDestroyed
  (for _array_element (Collision|GetOverlappingActors "/Game/Athena/PlayerPawn_Athena.PlayerPawn_Athena_C")
    (GameplayEffects|RemoveActiveGameplayEffectbySourceEffect (Ability|GetAbilitySystemComponent _array_element) "/Game/Athena/Prototype/Blueprints/Cube/GE_CubePositiveEffect.GE_CubePositiveEffect_C" 0 -1)))

(fn GetSocketClosestToTarget
  (params (TargetPlayer :type Actor%20Object%20Reference))
  (bind _returnvalue (Transformation|GetActorLocation TargetPlayer))
  (bind _closestsocket (LocalVariables|GetClosestSocket))
  (bind _ascube (Utilities|Casting|CastToCUBE (Actor|GetOwner)))
  (for _array_element (Transformation|GetAllSocketNames (Variables|/Game/Athena/Prototype/Blueprints/Cube/CUBE.CUBE_C|GetStaticMesh _ascube))
    (bind (_currentstep_t_16_ea4278f841ee4e73c6861e88cb99517b _currentstep_2parttransitiontime_12_209590664625dbbf3f6200a38992da89 _currentstep_downface_19_20869c6c4df1517848398693617bdfc2) (Utilities|Struct|BreakCubeStruct (Variables|/Game/Athena/Prototype/Blueprints/Cube/CUBE.CUBE_C|GetCurrentStep_T_16_EA4278F841EE4E73C6861E88CB99517B _ascube)))
    (if (Utilities|Name|NotEqual(Name) _array_element (Utilities|String|StringToName (Utilities|String|ToString(Byte) _currentstep_downface_19_20869c6c4df1517848398693617bdfc2)))
      (LocalVariables|SetDirectionTOPlayerTemp (- (Transformation|GetSocketLocation (Variables|/Game/Athena/Prototype/Blueprints/Cube/CUBE.CUBE_C|GetStaticMesh _ascube) _closestsocket) _returnvalue))
      (if (< (Math|Vector|VectorLength (- (Transformation|GetSocketLocation (Variables|/Game/Athena/Prototype/Blueprints/Cube/CUBE.CUBE_C|GetStaticMesh _ascube) _array_element) _returnvalue)) (Math|Vector|VectorLength (LocalVariables|GetDirectionTOPlayerTemp)))
        (LocalVariables|SetClosestSocket _array_element)
        (Variables|Default|SetDirectionToPlayer (LocalVariables|GetDirectionTOPlayerTemp)))))
  (return (Math|Vector|Normalize (- (Transformation|GetSocketLocation (Variables|/Game/Athena/Prototype/Blueprints/Cube/CUBE.CUBE_C|GetStaticMesh _ascube) _closestsocket) (Transformation|GetWorldLocation (Variables|/Game/Athena/Prototype/Blueprints/Cube/CUBE.CUBE_C|GetStaticMesh _ascube)))) (Transformation|GetSocketLocation (Variables|/Game/Athena/Prototype/Blueprints/Cube/CUBE.CUBE_C|GetStaticMesh _ascube) _closestsocket)))

(fn ConstructionScript
  (Components|Sphere|SetSphereRadius (Variables|Default|GetSphere) (ScalableFloat|GetValueatLevel (Variables|Default|GetShieldOverlapRadius)) true))
