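; Destination blueprint for the Cube that sinks into a lake and turns it into a bouncy, corrupted lake.
; On non-cosmetic instances it links to the CUBE actor and turns the lake on once setup reports the destination active.
; Cube events drive splash, sinking and steam effects.
; RepNotify functions attach sinking particles to the cube's down face, swap lake materials and audio, while UpdateCorruption feeds progress into the audio loop.
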(event EventBeginPlay
  (Networking|CanExecuteCosmeticEvents
    (:False
      (Variables|Default|SetIsCubeSinking 0)
      (Variables|Default|SetLakeState 0)
      (bind _outactors (Actor|GetAllActorsOfClass "/Game/Athena/Prototype/Blueprints/Cube/CUBE.CUBE_C"))
      (Utilities|FlowControl|Sequence
        (:then_0
          (Utilities|IsValid (Variables|Default|GetCubeRef)
            (:Is%20Not%20Valid
              (bind _output_get (Variables|Default|SetCubeRef (Utilities|Array|Get(acopy) _outactors)))
              (Variables|/Game/Athena/Prototype/Blueprints/Cube/CUBE.CUBE_C|SetCubeDestBlueprint self _output_get))))
        (:then_1
          (if (== (Variables|/Game/Athena/Prototype/Blueprints/Cube/CUBE.CUBE_C|GetPathIndex (Utilities|Array|Get(acopy) _outactors)) -1)
            (Default|BindEventtoCubeSetupComplete (Utilities|Array|Get(acopy) _outactors) (delegate Custom|CubeDone) :!class /Game/Athena/Prototype/Blueprints/Cube/CUBE.CUBE_C)
            (else
              (CallFunction|CubeDone 0))))))))

(event Custom|CubeDone
  (params (CurrentIndex :type Integer))
  (if (Variables|/Game/Athena/Prototype/Blueprints/Cube/CUBE.CUBE_C|GetDestTurnedOn (Utilities|Array|Get(acopy) (stale EventBeginPlay _outactors)))
    (CallFunction|LakeTurnOn)))

(event Custom|CubeSplash
  (bind _returnvalue (Effects|SpawnEmitteratLocation "/Game/Athena/Prototype/Blueprints/Cube/FX/P_Cube_Water_Roil.P_Cube_Water_Roil" "63296.000000,-49064.000000,-814.000000"))
  (bind _output_get (Variables|Default|SetSinkingWaterBoil _returnvalue))
  (bind _returnvalue_2 (Audio|SpawnSoundatLocation 0 (Transformation|GetWorldLocation _output_get)))
  (bind _output_get_2 (Variables|Default|SetPropagateAudioLoop _returnvalue_2))
  (Audio|Components|Audio|FadeIn _output_get_2 1.0 1.0 0.0 "Linear" :!class /Script/Engine.AudioComponent))

(event Custom|CubeStartSinking
  (Variables|Default|SetIsCubeSinking 1)
  (Variables|Default|SetLakeState 1))

(event Custom|CubeStopSinking
  (Utilities|FlowControl|Sequence
    (:then_0
      (Variables|Default|SetIsCubeSinking 0)
      (bind _propagateaudioloop (Variables|Default|GetPropagateAudioLoop)
        (:then
          (Audio|Components|Audio|FadeOut _propagateaudioloop 3.0 0.0 "Linear" :!class /Script/Engine.AudioComponent)
          (Audio|PlaySoundatLocation 0 (Transformation|GetWorldLocation _propagateaudioloop)))
        (:else)))
    (:then_1
      (Utilities|FlowControl|Delay 4.0)
      (Components|Activation|Deactivate (Variables|Default|GetSinkingWaterBoil)))
    (:then_2
      (Effects|SpawnEmitteratLocation "/Game/Athena/Prototype/Blueprints/Cube/FX/P_Cube_Finish_SteamRise.P_Cube_Finish_SteamRise" "63296.000000,-49064.000000,-814.000000"))))

(event Custom|LakeTurnOn
  (Variables|Default|SetLakeTurnedOn true)
  (Variables|Default|SetLakeState 2)
  (Effects|SpawnEmitteratLocation "/Game/Athena/Prototype/Blueprints/Cube/FX/P_C_L_TurnOn.P_C_L_TurnOn" (Variables|Default|GetLakeCenter))
  (Class|BPAthenaWater|Bouncy (Variables|Default|GetLakeActor))
  (Audio|SpawnSoundatLocation 0 (Variables|Default|GetLakeCenter)))

(fn ConvertLakeAudioActor
  (params (Lake :type BP%20Ambient%20Lake%20Audio%20Actor%20Object%20Reference))
  (Audio|Components|Audio|SetSound (Variables|/Game/Athena/Sounds/Ambient/BP_AmbientLakeAudioActor.BP_AmbientLakeAudioActor_C|GetWaterLoopAudio Lake) 0))

(fn CorruptRiverAudioActor
  (params (River :type BP%20Ambient%20Spline%20Actor%20Object%20Reference))
  (Utilities|FlowControl|Sequence
    (:then_0
      (Audio|Components|Audio|SetSound (Variables|/Script/FortniteGame.FortSplineAudioActor|GetAudioAtMidPoint River) 0))
    (:then_1
      (Audio|Components|Audio|SetSound (Variables|/Script/FortniteGame.FortSplineAudioActor|GetAudioAtClosestPoint River) 0))
    (:then_2
      (Audio|Components|Audio|SetSound (Variables|/Script/FortniteGame.FortSplineAudioActor|GetAudioAtNextClosestPoint River) 0))))

(fn OnRep_CubeSinking
  (bind (_currentstep_t_16_ea4278f841ee4e73c6861e88cb99517b _currentstep_2parttransitiontime_12_209590664625dbbf3f6200a38992da89 _currentstep_downface_19_20869c6c4df1517848398693617bdfc2) (Utilities|Struct|BreakCubeStruct (Variables|/Game/Athena/Prototype/Blueprints/Cube/CUBE.CUBE_C|GetCurrentStep_T_16_EA4278F841EE4E73C6861E88CB99517B (Variables|Default|GetCubeRef))))
  (bind _p_cube_printing_540 (Variables|Default|GetP_Cube_Printing_540))
  (if (Variables|Default|GetCubeSinking)
    (Transformation|AttachComponentToComponent _p_cube_printing_540 (Variables|/Game/Athena/Prototype/Blueprints/Cube/CUBE.CUBE_C|GetStaticMesh (Variables|Default|GetCubeRef)) (Utilities|String|StringToName (Utilities|String|ToString(Byte) _currentstep_downface_19_20869c6c4df1517848398693617bdfc2)) "SnapToTarget" "SnapToTarget" "KeepWorld" true)
    (Transformation|SetRelativeRotation _p_cube_printing_540 "180.000000, 0, 0" false false)
    (Components|Activation|SetActive _p_cube_printing_540 true false)
    (else
      (Components|Activation|SetActive (Variables|Default|GetP_Cube_Printing_540) false false))))

(fn OnRep_IsCubeSinking
  (bind _p_cube_printing_540 (Variables|Default|GetP_Cube_Printing_540))
  (switch int (Variables|Default|GetIsCubeSinking)
    (:0
      (Components|Activation|SetActive (Variables|Default|GetP_Cube_Printing_540) false false))
    (:1
      (bind _ascube (Utilities|Casting|CastToCUBE (Variables|Default|GetCubeRef)))
      (bind (_currentstep_t_16_ea4278f841ee4e73c6861e88cb99517b _currentstep_2parttransitiontime_12_209590664625dbbf3f6200a38992da89 _currentstep_downface_19_20869c6c4df1517848398693617bdfc2) (Utilities|Struct|BreakCubeStruct (Variables|/Game/Athena/Prototype/Blueprints/Cube/CUBE.CUBE_C|GetCurrentStep_T_16_EA4278F841EE4E73C6861E88CB99517B _ascube)))
      (Transformation|AttachComponentToComponent _p_cube_printing_540 (Variables|/Game/Athena/Prototype/Blueprints/Cube/CUBE.CUBE_C|GetStaticMesh _ascube) (Utilities|String|StringToName (Utilities|String|ToString(Byte) _currentstep_downface_19_20869c6c4df1517848398693617bdfc2)) "SnapToTarget" "SnapToTarget" "KeepWorld" true)
      (Transformation|SetRelativeRotation _p_cube_printing_540 "180.000000, 0, 0" false false)
      (Components|Activation|SetActive _p_cube_printing_540 true false))))

(fn OnRep_LakeState
  (bind _lakeactor (Variables|Default|GetLakeActor))
  (switch int (Variables|Default|GetLakeState)
    (:0
      (Class|BPAthenaWater|SetDefaultMaterial _lakeactor) ; Reset to default I guess
    )
    (:1
      (Class|BPAthenaWater|SetSpreadMaterial _lakeactor) ; Swap to corrupt material for spread
    )
    (:2
      (Class|BPAthenaWater|SetTurnedonMaterial _lakeactor) ; Set turned on material for bouncies
    )))

(fn OnRep_LakeTurnedOn
  (if (Variables|Default|GetLakeTurnedOn)
    (bind _propagateaudioloop (Variables|Default|GetPropagateAudioLoop)
      (:then
        (Audio|Components|Audio|FadeOut _propagateaudioloop 3.0 0.0 "Linear" :!class /Script/Engine.AudioComponent))
      (:else
        _))
    (Utilities|FlowControl|Sequence
      (:then_0
        (for _array_element (Variables|Default|GetShorelineAudioActors)
          (CallFunction|CorruptRiverAudioActor _array_element)))
      (:then_1
        (for _array_element_2 (Variables|Default|GetLakeAudioActors)
          (CallFunction|ConvertLakeAudioActor _array_element_2))))))

(fn UpdateCorruption
  (params (PercentComplete :type Float%20%28double-precision%29))
  (bind _propagateaudioloop (Variables|Default|GetPropagateAudioLoop)
    (:then
      (Audio|Parameter|SetFloatParameter _propagateaudioloop "Corruption" PercentComplete :!class /Script/AudioExtensions.AudioParameterControllerInterface)
      (Audio|Parameter|SetFloatParameter (Variables|Default|GetPropagateAudioLoop) "InvCorruption" (- 1.0 PercentComplete) :!class /Script/AudioExtensions.AudioParameterControllerInterface)
      (if (and (>= PercentComplete 0.83) (not (Variables|Default|GetPlayedPreSinkAudio)))
        (Audio|PlaySoundatLocation 0 (Transformation|GetWorldLocation (Variables|Default|GetPropagateAudioLoop)))
        (Variables|Default|SetPlayedPreSinkAudio true)))
    (:else)))

(fn ConstructionScript)
