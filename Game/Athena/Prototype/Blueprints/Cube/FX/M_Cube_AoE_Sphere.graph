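# Additive, unlit, two-sided material for the Cube AoE sphere effect, drawing an animated energy shell.
# Two noise-distorted, panning turbulence samples blend dark and light tints, plus a depth-based glow that stands alone on Low and Medium quality.
# Opacity combines depth fades, a UV band mask, camera-distance fades and inverted Fresnel; mobile skips one near-camera fade.
# A distance branch computes a huge downward offset beyond a horizontal camera-distance threshold.
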Material /Game/Athena/Prototype/Blueprints/Cube/FX/M_Cube_AoE_Sphere
graph {
 properties {blendMode=BLEND_Additive, pixelDepthOffsetMode=PDOM_Legacy, shadingModel=MSM_Unlit, twoSided=true, usedShadingModels=Unlit}
 nodes [
  n1 = WorldPosition() {worldPositionShaderOffset=WPT_CameraRelative}
  n2 = ScalarParameter() {defaultValue=3000, parameterName=FadeDist}
  n3 = Constant()
  If(
   A=MaterialFunctionCall(Vector 2=ComponentMask(None=@n1.XYZ) {g=true, r=true}) {materialFunction=/Engine/Functions/Engine_MaterialFunctions02/Utility/VectorLength}.V2 Length
   B=Multiply(A=@n2, B=ScalarParameter() {defaultValue=1.4, parameterName=HideDistanceBuffer}) {constB=1.36}
   "A > B"=Constant3Vector() {constant=<rgba 0 0 -8000000 0>}
   "A == B"=@n3
   "A < B"=@n3
  )
  n4 = Distance(A=ObjectPositionWS(), B=CameraPositionWS())
  n5 = ScalarParameter() {defaultValue=2500, parameterName=SphereRadius}
  n6 = ScalarParameter() {defaultValue=1400, parameterName=FadeDistance}
  n7 = Divide(A=Subtract(A=@n4, B=@n5), B=@n6)
  n8 = Multiply(A=@n5) {constB=0.5}
  Multiply(A=OneMinus(None=Clamp(None=@n7)), B=Divide(A=Subtract(A=@n4, B=@n8), B=@n6))
  n9 = VectorParameter() {parameterName=DarkColor, defaultValue=<rgba 0.1627 0.4326 1.8 0>}
  n10 = VectorParameter() {parameterName=LightColor, defaultValue=<rgba 1.1602 0.3695 3.5 1>}
  n11 = TextureCoordinate() {uTiling=9, vTiling=3}
  n12 = Multiply(A=Time(), B=ScalarParameter() {defaultValue=-0.3, parameterName=PanSpeed})
  n13 = Multiply(
   A=TextureSample(UVs=Panner(Coordinate=@n11, Time=@n12) {speedX=-0.02, speedY=-0.03}) {samplerType=SAMPLERTYPE_Grayscale, texture=&LowResBlurredNoise}.R
   B=Constant2Vector() {g=0.14, r=0.07}
  )
  n14 = Add(A=TextureCoordinate() {uTiling=6, vTiling=2}, B=Multiply(A=@n13) {constB=-1})
  n15 = Add(A=@n13, B=TextureCoordinate() {uTiling=8, vTiling=1.4})
  n16 = LinearInterpolate(
   A=TextureSample(UVs=Panner(Coordinate=@n14, Time=@n12) {speedX=-0.04, speedY=0.1}) {texture=/Game/Global/Textures/Default/Caustics/T_Turbulence_Seamless}.RGB
   B=TextureSample(UVs=Panner(Coordinate=@n15, Time=@n12) {speedX=0.05, speedY=0.06}) {texture=&T_Turbulence_Seamless}.RGB
  )
  n17 = ScalarParameter() {defaultValue=50, parameterName=DepthFade}
  n18 = Constant() {r=0.5}
  n19 = SphereMask(A=ComponentMask(None=TextureCoordinate()) {g=true}, B=@n18, Radius=@n18) {hardnessPercent=40}
  n20 = ScalarParameter() {defaultValue=120, parameterName=DepthFadeInverse}
  n21 = Add(A=OneMinus(None=DepthFade(FadeDistance=@n20)), B=ScalarParameter() {defaultValue=0.7, parameterName=OpacityAdd}) {constB=0.1}
  n22 = Power(
   Base=Multiply(A=DepthFade(FadeDistance=@n17) {fadeDistanceDefault=50}, B=Multiply(A=@n19, B=@n21) {constA=1}) {constA=1}
   Exp=ScalarParameter() {defaultValue=3, parameterName=DepthPower}
  )
  n23 = Multiply(A=@n10.RGB, B=Multiply(A=@n22) {constB=0.15})
  n24 = Add(A=Multiply(A=LinearInterpolate(A=@n9.RGB, B=@n10.RGB, Alpha=@n16), B=@n16), B=@n23)
  QualitySwitch(Default=@n24, Low=@n23, High=@n24, Medium=@n23)
  n25 = Clamp(None=LinearInterpolate(Alpha=MaterialFunctionCall(Distance=@n2) {materialFunction=&Near_Camera_Fade}.Result) {constA=1.25, constB=0})
  n26 = ScalarParameter() {defaultValue=1.2, parameterName=OpacityMult}
  n27 = Multiply(A=Multiply(A=@n22, B=Multiply(A=@n25, B=@n26)), B=OneMinus(None=Fresnel() {baseReflectFraction=0}))
  n28 = ScalarParameter() {defaultValue=500, parameterName=NearFadeOut}
  ShadingPathSwitch(
   Default=Multiply(A=@n27, B=MaterialFunctionCall(Distance=@n28) {materialFunction=&Near_Camera_Fade}.Result)
   Mobile=@n27
  ) {desc=Migrated from FeatureLevelSwitch}
 ]
}
