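# Translucent unlit particle material for glowing digital runes in the Cube effects, usable on Niagara and Cascade sprites.
# A rune mask, UV-distorted by grid-crack textures, blends dim and bright colors tinted by particle color into emissive.
# Opacity smoothsteps the powered mask like an SDF, with exponent from particle color or dynamic parameter, or a randomized sine under the Cube Skydive switch.
# Depth fade applies only on the default shading path.
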Material /Game/Athena/Prototype/Blueprints/Cube/FX/M_Cube_DigiRunes
graph {
 properties {
  bUsedWithNiagaraSprites=true
  bUsedWithParticleSprites=true
  blendMode=BLEND_Translucent
  pixelDepthOffsetMode=PDOM_Legacy
  shadingModel=MSM_Unlit
  usedShadingModels=Unlit
 }
 nodes [
  n1 = ParticleColor()
  n2 = ComponentMask(None=Transform(None=WorldPosition().XYZ) {transformSourceType=TRANSFORMSOURCE_World, transformType=TRANSFORM_Tangent}) {g=true, r=true}
  n3 = ScalarParameter() {defaultValue=300, parameterName=UVScale}
  n4 = Multiply(
   A=TextureSample(UVs=Divide(A=@n2, B=@n3)) {samplerType=SAMPLERTYPE_LinearColor, texture=/Game/Abilities/Player/Outlander/Actives/GroundWave/FX/T_GridCrack_Mask}.B
   B=TextureSample() {samplerType=SAMPLERTYPE_LinearColor, texture=&T_GridCrack_Mask}.G
  )
  n5 = ScalarParameter() {defaultValue=0.5, parameterName=Distort}
  n6 = TextureSampleParameter2D(UVs=Add(A=TextureCoordinate(), B=Multiply(A=@n4, B=@n5))) {parameterName=Texture, samplerType=SAMPLERTYPE_LinearColor, texture=/Game/Athena/Prototype/Blueprints/Cube/Materials/DFHeightMask}
  n7 = ParticleColor()
  n8 = DynamicParameter() {paramNames [Channel,Power,Param3,Param4]}
  n9 = ScalarParameter() {parameterName=ParticleColor}
  n10 = LinearInterpolate(A=@n6.R, B=@n6.R, Alpha=LinearInterpolate(A=@n7.R, B=@n8.Channel, Alpha=@n9))
  n11 = LinearInterpolate(
   A=VectorParameter() {parameterName=RuneColorDim}.RGB
   B=VectorParameter() {parameterName=RuneColorBright, defaultValue=<rgba 1.378 0.1161 2 0>}.RGB
   Alpha=@n10
  )
  MaterialFunctionCall(Emissive=LinearInterpolate(A=Multiply(A=@n1.G, B=@n11), B=Multiply(A=@n1.RGB, B=@n11), Alpha=ScalarParameter() {parameterName=ParticleColor})) {materialFunction=&MF_FX_EmissiveAdjustments}
  n12 = Add(A=Time(), B=ParticleRandom())
  n13 = Divide(A=Add(A=Sine(None=@n12)))
  n14 = Multiply(A=@n13) {constB=8}
  n15 = Power(Base=LinearInterpolate(A=@n6.R, B=@n6.R, Alpha=@n13), Exp=Ceil(None=Clamp(None=@n14) {maxDefault=8, minDefault=1}))
  n16 = Power(Base=@n10, Exp=LinearInterpolate(A=@n7.B, B=@n8.Power, Alpha=@n9))
  n17 = Multiply(
   A=@n1.A
   B=SmoothStep(
    Min=ScalarParameter() {parameterName=SDF Min}
    Max=ScalarParameter() {defaultValue=1, parameterName=SDF Max, sliderMax=1}
    Value=StaticSwitchParameter(True=@n15, False=@n16) {parameterName=Cube Skydive}
   ) {constMax=0.9, constMin=0.82}
  )
  ShadingPathSwitch(Default=DepthFade(Opacity=@n17, FadeDistance=ScalarParameter() {parameterName=Fade Distance}), Forward=@n17, Mobile=@n17)
 ]
}
