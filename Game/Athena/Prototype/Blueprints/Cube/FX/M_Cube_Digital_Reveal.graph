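# Translucent unlit particle material for a Cube digital reveal effect, drawing a cracked-grid pattern tinted by particle color.
# A radial smoke-texture mask, animated by dynamic Ripple and Power parameters, drives both brightness and visibility, and becomes a constant on mobile.
# The grid uses world position transformed to tangent space, with a bump-offset blurred glow dropped on low quality.
# Inverted depth fade confines the pattern to where it nears scene geometry.
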Material /Game/Athena/Prototype/Blueprints/Cube/FX/M_Cube_Digital_Reveal
graph {
 properties {
  bUsedWithMeshParticles=true
  bUsedWithNiagaraMeshParticles=true
  bUsedWithParticleSprites=true
  blendMode=BLEND_Translucent
  pixelDepthOffsetMode=PDOM_Legacy
  shadingModel=MSM_Unlit
  usedShadingModels=Unlit
 }
 nodes [
  n1 = ParticleColor()
  n2 = VectorParameter() {parameterName=BackgroundColor, defaultValue=<rgba 0.2272 0.3377 0.4792 0>}
  n3 = DynamicParameter() {defaultValue=<rgba 0.2 1 1 1>, paramNames [Ripple,Tile,Power,Param4]}
  n4 = LinearInterpolate(Alpha=@n3.Ripple) {constA=1, constB=-1}
  n5 = Constant2Vector() {g=1, r=0.15}
  n6 = MaterialFunctionCall(Float2=Add(
    A=Multiply(A=@n4, B=@n5)
    B=MaterialFunctionCall() {materialFunction=&VectorToRadialValue}.Radial Coordinates
   )) {materialFunction=&BreakOutFloat2Components}
  n7 = AppendVector(A=@n6.R, B=Clamp(None=@n6.G))
  n8 = ShadingPathSwitch(
   Default=Power(
    Base=TextureSampleParameter2D(UVs=@n7) {parameterName=MaskTexture, texture=/Game/Effects/Fort_Effects/Textures/Smoke/T_Smoke_Trail_Soft_2}.RGB
    Exp=@n3.Power
   )
   Mobile=Constant() {r=0.5}
  ) {desc=Migrated from FeatureLevelSwitch}
  MaterialFunctionCall(Emissive=Multiply(A=@n1.RGB, B=Add(A=@n2.RGB, B=@n8))) {materialFunction=&MF_FX_EmissiveAdjustments}
  MaterialFunctionCall() {materialFunction=&DebugTimeSine}
  n9 = WorldPosition()
  n10 = Divide(
   A=ComponentMask(None=Transform(None=@n9.XYZ) {transformSourceType=TRANSFORMSOURCE_World, transformType=TRANSFORM_Tangent}) {g=true, r=true}
   B=Multiply(A=ScalarParameter() {defaultValue=400, parameterName=UVScale}, B=@n3.Tile)
  )
  n11 = TextureSample(UVs=@n10) {samplerType=SAMPLERTYPE_LinearColor, texture=/Game/Abilities/Player/Outlander/Actives/GroundWave/FX/T_GridCrack_Mask}
  n12 = ScalarParameter() {defaultValue=-1, parameterName=BOffset}
  n13 = Multiply(
   A=TextureSample(UVs=BumpOffset(Coordinate=@n10, Height=@n12)) {texture=/Game/Abilities/Player/Outlander/Actives/GroundWave/FX/T_GridCrack_Mask_Blur}.R
   B=VectorParameter() {parameterName=BlurColor, defaultValue=<rgba 0.7119 1.2 1.0916 1>}.RGB
  )
  n14 = Constant()
  n15 = Multiply(
   A=Add(A=Add(A=@n11.R, B=@n11.R), B=QualitySwitch(Default=@n13, Low=@n14))
   B=TextureSample() {samplerType=SAMPLERTYPE_LinearColor, texture=&T_GridCrack_Mask}.G
  )
  n16 = ScalarParameter() {defaultValue=1, parameterName=Opacity}
  n17 = ScalarParameter() {defaultValue=70, parameterName=FadeDistance}
  Multiply(A=Multiply(A=Multiply(A=@n15, B=@n1.A), B=@n8), B=OneMinus(None=DepthFade(Opacity=@n16, FadeDistance=@n17) {fadeDistanceDefault=70}))
 ]
}
