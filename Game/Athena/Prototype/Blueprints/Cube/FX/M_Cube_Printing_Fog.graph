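# Additive unlit mesh-particle material for glowing pink-purple fog, apparently part of the Cube printing effect.
# It layers dark, smoky mid and bright colors using increasingly sharp powers of the V coordinate, tinted by particle color and remapped to HDR emissive.
# A panning swirl mask distorts the scrolling smoke texture, with a dynamic-parameter random time offset per particle.
# It also contains inverted depth-fade alpha, view-angle fade and normal-based vertex offset terms.
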Material /Game/Athena/Prototype/Blueprints/Cube/FX/M_Cube_Printing_Fog
graph {
 properties {bUsedWithMeshParticles=true, blendMode=BLEND_Additive, pixelDepthOffsetMode=PDOM_Legacy, shadingModel=MSM_Unlit, usedShadingModels=Unlit}
 nodes [
  LinearInterpolate(Alpha=MaterialFunctionCall("FadeSensitivity (optional)"=ScalarParameter() {defaultValue=3, parameterName=AngleFade}) {materialFunction=&FadeBasedOnViewAngle}.Result) {constA=1, constB=0.5}
  n1 = ComponentMask(None=TextureCoordinate()) {g=true}
  n2 = Multiply(A=VectorParameter() {parameterName=DarkColor, defaultValue=<rgba 0.2 0.0393 0.0755 1>}.RGB, B=Power(Base=@n1) {constExponent=3}) {constB=0.1}
  n3 = VectorParameter() {parameterName=SmokeColor, defaultValue=<rgba 0.8 0.2381 0.5691 1>}
  n4 = TextureCoordinate() {uTiling=5}
  n5 = TextureCoordinate() {uTiling=2, vTiling=0.25}
  n6 = Add(A=DynamicParameter() {paramNames [Random,Param2,Param3,Param4]}.Random, B=Time())
  n7 = Multiply(
   A=TextureSample(UVs=Panner(Coordinate=@n5, Time=@n6) {speedX=-0.01, speedY=-0.03}) {samplerType=SAMPLERTYPE_LinearGrayscale, texture=/Game/Global/Textures/Default/Flowmaps/T_StormSwirlMask}.R
   B=Constant2Vector() {g=0.05, r=0.1}
  ) {constB=0.1}
  n8 = TextureSample(UVs=Panner(Coordinate=Add(A=@n4, B=@n7), Time=@n6) {speedX=0.02, speedY=0.14}) {texture=&T_Inky_Smoke_Tile_Chromatic}
  n9 = Multiply(
   A=Add(A=Multiply(A=@n3.RGB, B=@n8.RGB), B=VectorParameter() {parameterName=MidColor, defaultValue=<rgba 0.2346 0.1346 0.55 0>}.RGB)
   B=Power(Base=@n1, Exp=ScalarParameter() {defaultValue=14, parameterName=MidPower}) {constExponent=5}
  ) {constB=0.25}
  n10 = Power(Base=@n1, Exp=ScalarParameter() {defaultValue=32, parameterName=BrightPower}) {constExponent=40}
  n11 = VectorParameter() {parameterName=BrightColor, defaultValue=<rgba 1 0.0748 0.91 0>}
  n12 = ParticleColor()
  n13 = Multiply(A=Add(A=Add(A=@n2, B=@n9), B=Multiply(A=@n10, B=@n11.RGB)), B=@n12.RGB)
  MaterialFunctionCall(Emissive=MaterialFunctionCall(LDR Input=Clamp(None=@n13), HDR Tint=@n13) {materialFunction=&DeriveHDRfromLDR}.RemappedRange) {materialFunction=&MF_FX_EmissiveAdjustments}
  n14 = DepthFade() {fadeDistanceDefault=120}
  Multiply(A=Multiply(A=@n12.A, B=OneMinus(None=@n14)))
  Multiply(A=VertexNormalWS(), B=ScalarParameter() {parameterName=VertexOffset})
 ]
}
