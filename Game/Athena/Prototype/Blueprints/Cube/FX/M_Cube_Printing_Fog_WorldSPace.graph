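# Additive unlit mesh-particle material for glowing, smoky fog in the prototype Cube's printing effect.
# Panning swirl mask distorts scrolling smoke, with time offset by a per-particle random value.
# A world-Z gradient anchored at LakeHeight, resized by the object's height above it, is raised to two powers for mid and bright color bands.
# Particle color tints the result before HDR emissive remapping; particle alpha times inverted depth fade favors geometry intersections.
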Material /Game/Athena/Prototype/Blueprints/Cube/FX/M_Cube_Printing_Fog_WorldSPace.M_Cube_Printing_Fog_WorldSpace
graph {
 properties {bUsedWithMeshParticles=true, blendMode=BLEND_Additive, pixelDepthOffsetMode=PDOM_Legacy, shadingModel=MSM_Unlit, usedShadingModels=Unlit}
 nodes [
  LinearInterpolate(Alpha=MaterialFunctionCall("FadeSensitivity (optional)"=ScalarParameter() {defaultValue=3, parameterName=AngleFade}) {materialFunction=&FadeBasedOnViewAngle}.Result) {constA=1, constB=0.5}
  n1 = VectorParameter() {parameterName=SmokeColor, defaultValue=<rgba 1.2 0.3572 0.8536 1>}
  n2 = TextureCoordinate() {uTiling=5}
  n3 = TextureCoordinate() {uTiling=2, vTiling=0.25}
  n4 = Add(A=DynamicParameter() {paramNames [Random,Param2,Param3,Param4]}.Random, B=Time())
  n5 = Multiply(
   A=TextureSample(UVs=Panner(Coordinate=@n3, Time=@n4) {speedX=-0.01, speedY=0.03}) {samplerType=SAMPLERTYPE_LinearGrayscale, texture=/Game/Global/Textures/Default/Flowmaps/T_StormSwirlMask}.R
   B=Constant2Vector() {g=0.05, r=0.1}
  ) {constB=0.1}
  n6 = TextureSample(UVs=Panner(Coordinate=Add(A=@n2, B=@n5), Time=@n4) {speedX=0.02, speedY=-0.16}) {texture=&T_Inky_Smoke_Tile_Chromatic}
  n7 = Add(
   A=Multiply(A=VertexColor().R, B=Multiply(A=@n1.RGB, B=@n6.RGB))
   B=VectorParameter() {parameterName=MidColor, defaultValue=<rgba 0.2346 0.1346 0.55 0>}.RGB
  )
  n8 = ScalarParameter() {defaultValue=-814, parameterName=LakeHeight}
  n9 = WorldPosition()
  n10 = ObjectPositionWS()
  n11 = ScalarParameter() {defaultValue=768, parameterName=CubeRadius}
  n12 = Distance(A=@n8, B=Add(A=ComponentMask(None=@n10) {b=true}, B=@n11))
  n13 = LinearInterpolate(A=ScalarParameter() {defaultValue=768, parameterName=GradientSize}, Alpha=Clamp(None=Divide(A=@n12, B=@n11))) {constA=1536, constB=300}
  n14 = ComponentMask(None=Subtract(A=@n8, B=Subtract(A=@n9.XYZ, B=@n13)) {constA=0, constB=-50}) {b=true}
  n15 = Multiply(B=Clamp(None=Divide(A=@n14, B=@n13))) {constA=1}
  n16 = Power(Base=@n15, Exp=ScalarParameter() {defaultValue=3, parameterName=MidPower}) {constExponent=5}
  n17 = Power(Base=@n15, Exp=ScalarParameter() {defaultValue=12, parameterName=BrightPower}) {constExponent=40}
  n18 = VectorParameter() {parameterName=BrightColor, defaultValue=<rgba 1 0.0748 0.91 0>}
  n19 = ParticleColor()
  n20 = Multiply(A=Add(A=Multiply(A=@n7, B=@n16) {constB=0.25}, B=Multiply(A=@n17, B=@n18.RGB)), B=@n19.RGB)
  MaterialFunctionCall(Emissive=MaterialFunctionCall(LDR Input=Clamp(None=@n20), HDR Tint=@n20) {materialFunction=&DeriveHDRfromLDR}.RemappedRange) {materialFunction=&MF_FX_EmissiveAdjustments}
  n21 = DepthFade() {fadeDistanceDefault=120}
  Multiply(A=Multiply(A=@n19.A, B=OneMinus(None=@n21)))
 ]
}
