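# Translucent, unlit, two-sided mesh particle material for a purple cube reveal effect.
# Panning dust and infected-map textures form a noise mask that blends two colors, with Fresnel tint and particle color for emissive.
# Opacity combines a noise-distorted world-space sphere mask, shifted vertically by Reveal, with a UV sphere mask and noise erosion.
# Dynamic parameters Reveal, SphereErode and Erode drive the animation, while particle alpha and camera depth fade scale opacity.
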Material /Game/Athena/Prototype/Blueprints/Cube/FX/M_Cube_Reveal
graph {
 properties {
  bUsedWithMeshParticles=true
  bUsedWithNiagaraMeshParticles=true
  blendMode=BLEND_Translucent
  pixelDepthOffsetMode=PDOM_Legacy
  shadingModel=MSM_Unlit
  twoSided=true
  usedShadingModels=Unlit
 }
 nodes [
  MaterialFunctionCall(Speed=ScalarParameter() {defaultValue=0.2, parameterName=DebugTime}) {materialFunction=&DebugTimeSine}
  MaterialFunctionCall!(
   CameraVector
   Texture=TextureObjectParameter() {parameterName=NoiseTexture, texture=&2ndInfectedMapLinear}
   Tiling=VectorParameter() {parameterName=NoiseTile, defaultValue=<rgba 3 3 3 1>}.RGB
   Offset
   AxisFadeContrast
  ) {materialFunction=/Engine/Functions/Engine_MaterialFunctions01/Texturing/TriplanarCameraVector}
  n1 = ScalarParameter() {defaultValue=1, parameterName=Tiling}
  n2 = Multiply(A=TextureCoordinate(), B=@n1)
  n3 = Power(Base=TextureSample(UVs=Panner(Coordinate=@n2) {speedX=0.025, speedY=-0.05}) {texture=/Game/Global/Textures/Default/PointSplatter/T_Tiling_Dust}.G) {constExponent=1.75}
  n4 = VectorParameter() {parameterName=ColorB, defaultValue=<rgba 0.1658 0.0117 0.5 1>}
  n5 = Multiply(A=Multiply(A=@n3, B=@n4.RGB) {constB=5}) {constB=5}
  n6 = Multiply(A=TextureCoordinate() {coordinateIndex=2}, B=@n1)
  n7 = AppendVector(A=ScalarParameter() {defaultValue=0.05, parameterName=SpeedX}, B=ScalarParameter() {defaultValue=0.2, parameterName=SpeedY})
  n8 = Panner(Coordinate=Multiply(A=@n1, B=TextureCoordinate() {coordinateIndex=2, uTiling=0.5, vTiling=0.5})) {speedX=0.05, speedY=0.2}
  n9 = Multiply(
   A=TextureSample(UVs=Panner(Coordinate=@n6, Speed=@n7) {speedX=0.05, speedY=0.2}) {texture=&2ndInfectedMapLinear}.G
   B=Add(A=@n3, B=TextureSample(UVs=@n8) {texture=&2ndInfectedMapLinear}.G)
  )
  n10 = Clamp(None=Power(Base=Multiply(A=@n9) {constB=4}) {constExponent=0.5})
  n11 = LinearInterpolate(A=VectorParameter() {parameterName=ColorA, defaultValue=<rgba 2.6587 0.3125 3 1>}.RGB, B=@n4.RGB, Alpha=@n10)
  n12 = LinearInterpolate(
   A=@n11
   B=Multiply(A=@n11, B=VectorParameter() {parameterName=PurpleTint, defaultValue=<rgba 1.4167 0.2493 5 1>}.RGB)
   Alpha=Power(Base=Fresnel() {exponent=3})
  )
  MaterialFunctionCall(Emissive=Multiply(A=ParticleColor().RGB, B=Add(A=@n5, B=@n12))) {materialFunction=&MF_FX_EmissiveAdjustments}
  n13 = Multiply(A=@n10) {constB=50}
  n14 = WorldPosition()
  n15 = ParticlePositionWS()
  n16 = Multiply(
   A=LinearInterpolate(Alpha=DynamicParameter() {defaultValue=<rgba 1 1 0.75 1>, paramNames [Reveal,SphereErode,Erode,Param4]}.Reveal) {constA=-3000, constB=0}
   B=Constant3Vector() {constant=<rgba 0 0 1 0>}
  )
  n17 = Add(A=Multiply(A=ComponentMask(None=@n10) {g=true, r=true}) {constB=0.15}, B=TextureCoordinate() {coordinateIndex=2})
  n18 = Constant2Vector() {desc=SphereOffset, r=-0.15}
  n19 = OneMinus(None=SphereMask(A=Add(A=@n17, B=@n18), B=Constant2Vector() {g=0.5, r=0.5}, Radius=Constant() {r=1}, Hardness=Constant()))
  n20 = DynamicParameter() {defaultValue=<rgba 1 1 0.75 1>, paramNames [Reveal,SphereErode,Erode,Param4]}
  n21 = LinearInterpolate(Alpha=@n20.SphereErode) {constA=-1}
  n22 = Multiply(
   A=SphereMask(
    A=Add(A=@n13, B=@n14.XYZ)
    B=Add(A=@n15, B=@n16)
    Radius=ScalarParameter() {defaultValue=2037.2413, parameterName=Radius}
    Hardness=Constant() {r=0.5}
   )
   B=Clamp(None=Add(A=@n19, B=@n21))
  )
  n23 = Clamp(None=Add(A=@n10, B=LinearInterpolate(Alpha=@n20.Erode) {constA=-1}))
  Multiply(
   A=Multiply(A=ParticleColor().A, B=Multiply(A=@n22, B=@n23))
   B=MaterialFunctionCall(Fade Length=ScalarParameter() {parameterName=CameraFadeDis}) {materialFunction=&CameraDepthFade}.Result
  )
  ObjectPositionWS()
  ScalarParameter() {defaultValue=1, parameterName=Reveal}
 ]
}
