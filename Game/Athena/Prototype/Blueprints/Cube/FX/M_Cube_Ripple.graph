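# M_Cube_Ripple is a translucent, unlit particle material that draws animated, refracting circular ripples for the Cube FX.
# Radial sine waves, driven by dynamic Ripple and Distortion parameters and roughened by radial noise, blend a sphere normal that lights a tinted cubemap reflection and sun-direction highlight.
# An expanding ring mask and optional depth-fade mask confine the refraction, which a shading path switch flattens on forward and mobile.
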Material /Game/Athena/Prototype/Blueprints/Cube/FX/M_Cube_Ripple
graph {
 properties {
  bUsedWithMeshParticles=true
  bUsedWithNiagaraMeshParticles=true
  bUsedWithNiagaraSprites=true
  blendMode=BLEND_Translucent
  pixelDepthOffsetMode=PDOM_Legacy
  refractionMethod=RM_IndexOfRefraction
  shadingModel=MSM_Unlit
  usedShadingModels=Unlit
 }
 nodes [
  Constant() {r=0.5}
  n1 = CollectionParameter() {
   collection=&FortniteMaterialParameters
   parameterName=SunAndMoonModelDirectionalVector
  }
  n2 = Constant3Vector() {constant=<rgba 0 0 1 0>}
  n3 = TextureSample() {samplerType=SAMPLERTYPE_Normal, texture=&T_Sphere_N}
  n4 = ParticleColor()
  n5 = Power(Base=Distance(A=Constant() {r=0.5}, B=TextureCoordinate())) {constExponent=1.4}
  n6 = NamedRerouteUsage()
  n7 = LinearInterpolate(
   A=ScalarParameter() {defaultValue=6, parameterName=NumRipplesStart}
   B=ScalarParameter() {defaultValue=8, parameterName=NumRipplesEnd}
   Alpha=@n6
  ) {constA=12, constB=3}
  n8 = TextureCoordinate()
  n9 = TextureSample!(
   UVs=Multiply(
    A=MaterialFunctionCall() {materialFunction=&VectorToRadialValue}.Radial Coordinates
    B=Constant2Vector() {g=0.3, r=2}
   )
   Tex
   "DDX(UVs)"=DDX(Value=@n8)
   "DDY(UVs)"=DDY(Value=@n8)
   Apply View MipBias
  ) {
   mipValueMode=TMVM_Derivative
   samplerType=SAMPLERTYPE_Normal
   texture=/Game/Characters/Player/Female/Medium/Bodies/F_MED_Bistro_Astronaut/FX/T_GaussianNoise_Normal
  }
  n10 = LinearInterpolate(Alpha=@n9.R) {constA=0.8}
  n11 = NamedRerouteUsage()
  n12 = Multiply(A=Add(A=@n6, B=LinearInterpolate(B=@n10, Alpha=@n11))) {constB=-1}
  n13 = ScalarParameter() {defaultValue=6, parameterName=RippleSpeed}
  n14 = Sine(None=Add(A=Multiply(A=@n5, B=@n7) {constB=2}, B=Multiply(A=@n12, B=@n13) {constB=-5}))
  n15 = Constant() {r=0.5}
  n16 = SphereMask(A=TextureCoordinate(), B=@n15, Radius=@n15) {attenuationRadius=0.5, hardnessPercent=0}
  n17 = Multiply(A=@n4.A, B=Power(Base=ConstantBiasScale(None=@n14), Exp=LinearInterpolate(Alpha=@n16) {constA=1.5, constB=0.7}) {constExponent=1.5})
  n18 = ReflectionVectorWS(CustomWorldNormal=Transform(None=LinearInterpolate(A=@n2, B=@n3.RGB, Alpha=@n17)))
  n19 = DotProduct(A=ComponentMask(None=@n1) {b=true, g=true, r=true}, B=@n18)
  n20 = ScalarParameter() {defaultValue=25, parameterName=Spec Highlight Power}
  n21 = Multiply(
   A=Clamp(None=Power(Base=@n19, Exp=@n20) {constExponent=20})
   B=VectorParameter() {parameterName=SpecHighlightTint, defaultValue=<rgba 0.3674 0.2447 0.5 0>}.RGB
  )
  n22 = TextureSampleParameterCube!(UVs=@n18, Apply View MipBias) {
   parameterName=MaterialExpressionTextureSampleParameterCube_0
   samplerType=SAMPLERTYPE_LinearColor
   texture=/Game/Packages/Fortress_Sky/TexturesHDR/T_AthenaSkylight
  }
  n23 = Desaturation(
   None=Multiply(A=@n22.RGB, B=VectorParameter() {parameterName=Cube Map Multiplier, defaultValue=<rgba 0.0577 0.0375 0.1 0>}.RGB) {constB=1.5}
   Fraction=Constant() {r=0.6}
  )
  MaterialFunctionCall(Emissive=Multiply(A=Add(A=@n21, B=@n23), B=@n4.RGB)) {materialFunction=&MF_FX_EmissiveAdjustments}
  n24 = DynamicParameter() {defaultValue=<rgba 0.2 1 1 1>, paramNames [Ripple,Crack,Distortion,Param4]}
  n25 = MaterialFunctionCall() {materialFunction=&DebugTimeSine}
  Multiply(A=@n24.Ripple, B=@n25.Result)
  Multiply(A=@n24.Distortion, B=OneMinus(None=@n25.Result))
  NamedRerouteDeclaration(None=@n24.Ripple) {name=Ripple, nodeColor=<rgba 0.7109 0 1 1>}
  n26 = ScalarParameter() {parameterName=UseDistortion}
  NamedRerouteDeclaration(None=Multiply(A=@n24.Distortion, B=LinearInterpolate(Alpha=@n26))) {name=Distortion, nodeColor=<rgba 0.6406 0 1 1>}
  n27 = TextureCoordinate()
  n28 = Constant() {r=0.5}
  n29 = Multiply(A=LinearInterpolate(B=LinearInterpolate(Alpha=@n9.R) {constA=0.5, constB=1.5}, Alpha=NamedRerouteUsage()) {constA=1}, B=NamedRerouteUsage())
  n30 = Power(Base=@n29) {constExponent=0.7}
  n31 = Clamp(None=LinearInterpolate(Alpha=@n29) {constA=-0.7})
  n32 = Multiply(
   A=SphereMask(A=@n27, B=@n28, Radius=Clamp(None=@n30) {maxDefault=0.5}) {attenuationRadius=0.5, hardnessPercent=88}
   B=OneMinus(None=SphereMask(A=@n27, B=@n28, Radius=@n31) {hardnessPercent=50})
  )
  n33 = OneMinus(None=DepthFade(FadeDistance=ScalarParameter() {defaultValue=70, parameterName=FadeDistance}) {fadeDistanceDefault=70})
  n34 = ScalarParameter() {defaultValue=1, parameterName=UseDepthFadeMask}
  n35 = LinearInterpolate(
   B=ScalarParameter() {defaultValue=1.1, parameterName=RefractionAmount}
   Alpha=Multiply(A=Multiply(A=@n17, B=@n32), B=LinearInterpolate(B=@n33, Alpha=@n34) {constA=1})
  ) {constA=1, constB=1.1}
  n36 = Constant() {r=1}
  ShadingPathSwitch(Default=@n35, Deferred=@n35, Forward=@n36, Mobile=@n36)
 ]
}
