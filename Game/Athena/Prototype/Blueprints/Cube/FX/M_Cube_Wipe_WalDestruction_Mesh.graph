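# Translucent unlit mesh-particle material for the Cube wall-destruction wipe effect, drawing animated cartoon flames.
# A panning normal map distorts the UVs of two scrolling flame layers and two sine-warped fire masks.
# Flames plus a vertical falloff, multiplied by the masks, drive particle-colored HDR emissive and vertically faded alpha, with depth fade skipped on mobile.
# The dynamic parameter named Refraction actually sets the flame textures' mip bias.
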Material /Game/Athena/Prototype/Blueprints/Cube/FX/M_Cube_Wipe_WalDestruction_Mesh
graph {
 properties {bUsedWithMeshParticles=true, blendMode=BLEND_Translucent, pixelDepthOffsetMode=PDOM_Legacy, shadingModel=MSM_Unlit, usedShadingModels=Unlit}
 nodes [
  n1 = ScalarParameter() {defaultValue=1, parameterName=UScale}
  n2 = ScalarParameter() {defaultValue=1, parameterName=VScale}
  n3 = TextureSample(UVs=Panner(Coordinate=TextureCoordinate() {coordinateIndex=1}) {speedX=1, speedY=0.1}) {
   samplerSource=SSM_Wrap_WorldGroupSettings
   samplerType=SAMPLERTYPE_Normal
   texture=/Game/Effects/Fort_Effects/Gameplay/SmokeDustExplosions/Textures/T_Explosion_NM_01
  }
  n4 = ComponentMask(None=MaterialFunctionCall(Normal=@n3.RGB, Flatness=ScalarParameter() {defaultValue=1, parameterName=UVDist}) {materialFunction=&FlattenNormal}.Result) {g=true, r=true}
  n5 = Add(A=Multiply(A=TextureCoordinate() {uTiling=2}, B=AppendVector(A=@n1, B=@n2)), B=@n4)
  n6 = DynamicParameter() {defaultValue=<rgba 1 1 0.5 1>, paramNames [FresnelExponent, Refraction, GlowPan, Param4]}
  n7 = TextureSample!(UVs=Panner(Coordinate=Add(A=@n5) {constB=0.37}) {speedX=-0.25, speedY=-1}, Tex, Bias=@n6.Refraction, Apply View MipBias) {
   mipValueMode=TMVM_MipBias
   samplerSource=SSM_Wrap_WorldGroupSettings
   texture=/Game/Effects/Fort_Effects/Textures/Tile/T_Tile_Texture_Cartoon_Flames_Layer
  }
  n8 = TextureSample(UVs=Panner(Coordinate=@n5) {speedX=-0.2, speedY=-1.2}, Bias=@n6.Refraction) {
   mipValueMode=TMVM_MipBias
   samplerSource=SSM_Wrap_WorldGroupSettings
   texture=&T_Tile_Texture_Cartoon_Flames_Layer
  }
  n9 = Add(A=Multiply(A=@n7.G) {constB=0.5}, B=Multiply(A=@n8.G) {constB=0.5})
  n10 = TextureCoordinate()
  n11 = Power(Base=OneMinus(None=ComponentMask(None=@n10) {g=true})) {constExponent=10}
  n12 = Add(A=TextureCoordinate(), B=@n4)
  n13 = OneMinus(None=ComponentMask(None=@n12) {g=true})
  n14 = OneMinus(None=ComponentMask(None=@n12) {g=true})
  n15 = TextureCoordinate()
  n16 = LinearInterpolate(
   A=AppendVector(A=ComponentMask(None=@n12) {r=true}, B=Add(A=@n13) {constB=-0.15})
   B=AppendVector(A=ComponentMask(None=@n12) {r=true}, B=Add(A=@n14) {constB=0.45})
   Alpha=Sine(None=ComponentMask(None=@n15) {r=true}) {period=2}
  )
  n17 = ScalarParameter() {defaultValue=2, parameterName=Bias_A}
  n18 = TextureSample(UVs=Add(A=Panner(Coordinate=@n16) {speedX=-0.25}, B=TextureCoordinate() {uTiling=0, vTiling=0.15}), Bias=@n17) {mipValueMode=TMVM_MipBias, texture=/Game/Effects/Fort_Effects/Textures/Fire/T_Fire_Mask}
  n19 = Add(A=TextureCoordinate(), B=@n4)
  n20 = OneMinus(None=ComponentMask(None=@n19) {g=true})
  n21 = OneMinus(None=ComponentMask(None=@n19) {g=true})
  n22 = TextureCoordinate()
  n23 = LinearInterpolate(
   A=AppendVector(A=ComponentMask(None=@n19) {r=true}, B=Add(A=@n20) {constB=-0.15})
   B=AppendVector(A=ComponentMask(None=@n19) {r=true}, B=Add(A=@n21) {constB=0.45})
   Alpha=Sine(None=ComponentMask(None=@n22) {r=true}) {period=2}
  )
  n24 = TextureSample(UVs=Add(A=Panner(Coordinate=@n23) {speedX=-0.25}, B=TextureCoordinate()), Bias=@n17) {mipValueMode=TMVM_MipBias, texture=&T_Fire_Mask}
  n25 = Clamp(None=Multiply(A=Add(A=@n9, B=@n11), B=Multiply(A=@n18.R, B=@n24.R)))
  n26 = TextureCoordinate() {vTiling=6}
  n27 = TextureCoordinate()
  n28 = TextureSample!(
   UVs=LinearInterpolate(
    A=Add(A=@n26) {constB=-5.5}
    B=Add(A=@n27) {constB=0.7}
    Alpha=DynamicParameter() {defaultValue=<rgba 1 1 0.5 1>, paramNames [FresnelExponent, Refraction, GlowPan, Param4]}.GlowPan
   )
   Tex
   Apply View MipBias
  ) {texture=/Game/Global/Textures/Default/Masks_and_Gradients/T_Vert_Gradient_Soft}
  Add(A=@n25, B=Multiply(A=Power(Base=@n28.R) {constExponent=1}) {constB=2}) {constA=1}
  n29 = ParticleColor()
  n30 = Multiply(A=Multiply(A=@n29.RGB, B=@n25)) {constB=0.5}
  MaterialFunctionCall(Emissive=MaterialFunctionCall(LDR Input=Clamp(None=@n30), HDR Tint=@n29.RGB, HDR Intensity Multiplier=ScalarParameter() {defaultValue=1, parameterName=HDRMult}) {materialFunction=&DeriveHDRfromLDR}.RemappedRange) {materialFunction=&MF_FX_EmissiveAdjustments}
  n31 = Power(Base=ComponentMask(None=TextureCoordinate()) {g=true})
  n32 = Multiply(A=Multiply(A=@n25, B=ParticleColor().A), B=LinearInterpolate(A=Constant(), B=Constant() {r=1}, Alpha=Clamp(None=@n31)))
  ShadingPathSwitch(Default=Multiply(A=@n32, B=DepthFade()), Mobile=@n32) {desc=Migrated from FeatureLevelSwitch}
 ]
}
