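# Additive, unlit, two-sided mesh-particle material for glowing light sheets in the Cube prototype's effects.
# A vertical UV gradient mask, softened by vertex red, shapes a secondary color plus two panning noise layers with a tint, all multiplied by particle color.
# Opacity combines particle alpha, view-angle fade and depth fade.
# A horizontal camera-distance lerp supplies an offset that sinks sheets downward as the camera nears.
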Material /Game/Athena/Prototype/Blueprints/Cube/FX/M_Light_Sheets
graph {
 properties {
  bUsedWithMeshParticles=true
  blendMode=BLEND_Additive
  pixelDepthOffsetMode=PDOM_Legacy
  shadingModel=MSM_Unlit
  twoSided=true
  usedShadingModels=Unlit
 }
 nodes [
  n1 = CameraPositionWS()
  n2 = WorldPosition()
  n3 = Add(
   A=Distance(A=ComponentMask(None=@n1) {g=true, r=true}, B=ComponentMask(None=@n2.XYZ) {g=true, r=true})
   B=ScalarParameter() {defaultValue=-500, parameterName=SinkOfsset}
  )
  n4 = ScalarParameter() {defaultValue=5000, parameterName=SinkDistance}
  LinearInterpolate(A=VectorParameter() {parameterName=SinkAmount, defaultValue=<rgba 0 0 -4000 0>}.RGB, Alpha=Clamp(None=Divide(A=@n3, B=@n4))) {constB=0}
  n5 = ParticleColor()
  n6 = ComponentMask(None=TextureCoordinate()) {g=true}
  n7 = VertexColor()
  n8 = Multiply(
   A=SphereMask(A=@n6, B=Constant() {r=0.5}, Hardness=LinearInterpolate(Alpha=@n7.R) {constA=0.8, constB=0.2}) {attenuationRadius=0.5, hardnessPercent=80}
   B=Power(Base=@n6) {constExponent=3}
  )
  n9 = Multiply(A=@n8, B=VectorParameter() {parameterName=SecondaryColor, defaultValue=<rgba 0.0908 0.0875 0.4 0>}.RGB) {constB=0.1}
  n10 = TextureSample(UVs=Panner(Coordinate=TextureCoordinate() {uTiling=16, vTiling=0.35}) {speedX=0.02, speedY=0.4}) {samplerType=SAMPLERTYPE_Grayscale, texture=&LowResBlurredNoise}
  n11 = TextureSample(UVs=Panner(Coordinate=TextureCoordinate() {uTiling=5, vTiling=0.45}) {speedX=-0.02, speedY=0.1}) {samplerType=SAMPLERTYPE_Grayscale, texture=&LowResBlurredNoise}
  n12 = Multiply(A=Multiply(A=@n8, B=Multiply(A=@n10.G, B=@n11.G)), B=VectorParameter() {parameterName=TintColor, defaultValue=<rgba 1 1 1 1>}.RGB)
  MaterialFunctionCall(Emissive=Multiply(A=@n5.RGB, B=Add(A=@n9, B=@n12))) {materialFunction=&MF_FX_EmissiveAdjustments}
  n13 = CameraVectorWS()
  n14 = VertexNormalWS()
  n15 = Power(Base=Abs(None=DotProduct(A=@n13, B=@n14)), Exp=ScalarParameter() {defaultValue=3, parameterName=AngleFade})
  n16 = ScalarParameter() {defaultValue=400, parameterName=DepthFade}
  Multiply(A=@n5.A, B=Multiply(A=Clamp(None=@n15), B=DepthFade(FadeDistance=@n16)))
 ]
}
