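# Unlit additive material for the distant light-sheet effect among the prototype Cube FX.
# Panning blurred noise and a vertical UV gradient mask combine with primary and secondary colors to form the glow.
# View-angle falloff pairs with camera-distance and depth fades, so sheets appear only from afar and soften at intersections.
# A Ceil-based switch yields a large downward offset when the camera is inside the fade distance, apparently moving nearby sheets out of view.
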Material /Game/Athena/Prototype/Blueprints/Cube/FX/M_Light_Sheets_Distant
graph {
 properties {blendMode=BLEND_Additive, pixelDepthOffsetMode=PDOM_Legacy, shadingModel=MSM_Unlit, usedShadingModels=Unlit}
 nodes [
  ActorPositionWS()
  n1 = CameraVectorWS()
  n2 = VertexNormalWS()
  n3 = Clamp(None=Power(Base=DotProduct(A=@n1, B=@n2), Exp=ScalarParameter() {defaultValue=4, parameterName=AngleFade}))
  n4 = ObjectPositionWS()
  n5 = CameraPositionWS()
  n6 = Divide(
   A=Subtract(A=Distance(A=@n4, B=@n5), B=ScalarParameter() {defaultValue=30000, parameterName=FadeDistanceOffset})
   B=ScalarParameter() {defaultValue=12000, parameterName=FadeDistance}
  )
  Multiply(A=@n3, B=Clamp(None=@n6))
  Multiply(A=@n3, B=DepthFade(FadeDistance=ScalarParameter() {defaultValue=2000, parameterName=DepthFade}))
  n7 = Ceil(None=@n6)
  Multiply(A=OneMinus(None=Clamp(None=@n7)), B=VectorParameter() {parameterName=VerticalOffset, defaultValue=<rgba 0 0 -100000 0>}.RGB)
  n8 = ComponentMask(None=TextureCoordinate()) {g=true}
  n9 = Multiply(A=SphereMask(A=@n8, B=Constant() {r=0.5}) {attenuationRadius=0.5, hardnessPercent=80}, B=Power(Base=@n8))
  n10 = VectorParameter() {parameterName=SecondaryColor, defaultValue=<rgba 0.0908 0.0875 0.4 0>}
  n11 = TextureCoordinate() {uTiling=6, vTiling=0.45}
  n12 = Add(A=TextureSample(UVs=Panner(Coordinate=@n11) {speedX=-0.03, speedY=0.1}) {samplerType=SAMPLERTYPE_Grayscale, texture=&LowResBlurredNoise}.G) {constB=0.1}
  Multiply(
   A=VectorParameter() {parameterName=PrimaryColor, defaultValue=<rgba 0.848 0.1602 1.2 0>}.RGB
   B=Add(A=Multiply(A=@n9, B=@n10.RGB) {constB=0.1}, B=Multiply(A=@n9, B=@n12))
  )
 ]
}
