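# Additive, unlit, two-sided mesh-particle material for the prototype Cube's magnetic field dome effect, drawing animated, glowing strands.
# SplineThicken sets strand width from a sine pulse offset by UV1 and dynamic-parameter Width, while panning noise adds world-position wobble.
# Two panning smoke samples, simplified on low quality, modulate edge-to-center colors.
# Opacity combines scrolling gradient sections, end fades, the pulse, view-facing falloff, depth fade (non-mobile) and time-of-day dimming.
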Material /Game/Athena/Prototype/Blueprints/Cube/FX/M_Magnetic_Field_Dome
graph {
 properties {
  bUsedWithMeshParticles=true
  blendMode=BLEND_Additive
  pixelDepthOffsetMode=PDOM_Legacy
  shadingModel=MSM_Unlit
  twoSided=true
  usedShadingModels=Unlit
 }
 nodes [
  n1 = TextureCoordinate()
  n2 = Constant() {r=0.5}
  n3 = SphereMask(A=ComponentMask(None=@n1) {g=true}, B=@n2, Radius=@n2) {hardnessPercent=50}
  n4 = TextureCoordinate() {coordinateIndex=1}
  n5 = Multiply(A=LinearInterpolate(Alpha=ComponentMask(None=@n4) {g=true}) {constA=0.4, constB=0.8}, B=Time())
  n6 = LinearInterpolate(Alpha=ConstantBiasScale(None=Sine(None=@n5) {period=4.75})) {constA=0.1}
  n7 = ScalarParameter() {defaultValue=200, parameterName=Width}
  n8 = DynamicParameter() {paramNames [Width,Random,Pan,Param4]}
  n9 = Multiply(A=@n3, B=Multiply(A=@n6, B=Multiply(A=@n7, B=@n8.Width)))
  n10 = MaterialFunctionCall(WidthBase=@n9, WidthTip=@n9) {materialFunction=&SplineThicken}
  n11 = ComponentMask(None=TextureCoordinate()) {g=true}
  n12 = Add(A=TextureCoordinate() {uTiling=0}, B=TextureCoordinate() {coordinateIndex=1})
  n13 = ScalarParameter() {defaultValue=0.7, parameterName=WobbleTile}
  n14 = Add(A=Time(), B=@n8.Random)
  n15 = ScalarParameter() {defaultValue=0.15, parameterName=WobbleSpeed}
  n16 = TextureSample(UVs=Panner(Coordinate=Multiply(A=@n12, B=@n13), Time=Multiply(A=@n14, B=@n15)) {speedY=0.1}) {samplerType=SAMPLERTYPE_LinearColor, texture=&HighContrastNoise}
  n17 = ScalarParameter() {defaultValue=400, parameterName=WobbleStrength}
  Add(A=@n10.WorldPositionOffset, B=Multiply(A=LinearInterpolate(Alpha=@n11) {constA=0.5}, B=Multiply(A=@n16.RGB, B=@n17)))
  n18 = VectorParameter() {parameterName=BackgroundColor, defaultValue=<rgba 0.234 0.3491 1.5 0>}
  n19 = Add(A=Time(), B=@n8.Random)
  n20 = Panner(Coordinate=TextureCoordinate() {uTiling=0.15, vTiling=2.5}, Time=@n19) {speedX=0.025, speedY=-0.03}
  n21 = TextureCoordinate() {coordinateIndex=2}
  n22 = Panner(Coordinate=TextureCoordinate() {uTiling=0.25, vTiling=4}, Time=@n19) {speedX=-0.04, speedY=-0.1}
  n23 = Add(
   A=TextureSample(UVs=Add(A=@n20, B=@n21)) {texture=&T_Inky_Smoke_Tile_Chromatic}.RGB
   B=TextureSample(UVs=Add(A=@n21, B=@n22)) {texture=&T_Inky_Smoke_Tile_Chromatic}.RGB
  )
  n24 = QualitySwitch(Default=@n23, Low=Constant() {r=0.5}, High=@n23, Medium=@n23)
  n25 = Multiply(
   A=LinearInterpolate(
    A=VectorParameter() {parameterName=Color, defaultValue=<rgba 2 0.1979 1.9138 0>}.RGB
    B=VectorParameter() {parameterName=OuterColor, defaultValue=<rgba 0.3028 0.8256 2 0>}.RGB
    Alpha=@n3
   )
   B=@n24
  )
  n26 = ParticleColor()
  MaterialFunctionCall(Emissive=Multiply(A=Add(A=@n18.RGB, B=@n25), B=@n26.RGB)) {materialFunction=&MF_FX_EmissiveAdjustments}
  n27 = Constant() {r=0.65}
  n28 = Add(A=ComponentMask(None=@n4) {r=true}, B=@n5)
  n29 = TextureSample(UVs=Panner(Coordinate=TextureCoordinate() {vTiling=0.2}, Time=Add(A=@n8.Random, B=@n28)) {speedY=-0.04}) {texture=/Game/Global/Textures/Default/Masks_and_Gradients/T_Vert_Gradient_Sections}
  n30 = ComponentMask(None=@n1) {r=true}
  n31 = Multiply(
   A=QualitySwitch(Default=Multiply(A=@n29.R, B=@n3), Low=Constant() {r=1})
   B=Power(Base=SphereMask(A=@n30, B=@n2, Radius=@n2) {hardnessPercent=0})
  )
  n32 = CameraVectorWS()
  n33 = Multiply(A=Multiply(A=@n24, B=Multiply(A=@n31, B=@n6)), B=Clamp(None=DotProduct(A=@n32, B=@n10.Normal)))
  n34 = ShadingPathSwitch(Default=DepthFade() {fadeDistanceDefault=200}, Mobile=Constant() {r=1}) {desc=Migrated from FeatureLevelSwitch}
  Multiply(
   A=MaterialFunctionCall(Color_Morning=@n27, Color_Day=Constant() {r=1}, Color_Evening=@n27, Color_Night=Constant() {r=0.45}) {materialFunction=&Color_Time_of_Day}.Result
   B=Multiply(A=@n26.A, B=Multiply(A=@n33, B=@n34))
  )
 ]
}
