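# Lit translucent smoke material for particle and Niagara sprites in the Cube prototype's effects, built on an 8x8 smoke flipbook.
# Color is SmokeColor times particle color; opacity multiplies flipbook alpha, tiled inky noise with dynamic offset and scale, particle alpha and near-camera fade.
# A depth flipbook varies the soft-intersection DepthFade distance.
# Flipbook normals, flattened by dynamic Param4, blend with a per-frame tiled sphere normal for rounded volumetric lighting.
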Material /Game/Athena/Prototype/Blueprints/Cube/FX/M_SwirlSmoke_01
graph {
 properties {
  allowTranslucentCustomDepthWrites=true
  bUsedWithNiagaraSprites=true
  bUsedWithParticleSprites=true
  blendMode=BLEND_Translucent
  opacityMaskClipValue=0.05
  pixelDepthOffsetMode=PDOM_Legacy
  translucencyLightingMode=TLM_VolumetricPerVertexDirectional
  usedShadingModels=Default Lit
 }
 nodes [
  n1 = ParticleColor()
  n2 = DynamicParameter() {paramNames [OffsetUVs, NoiseScale, MaxFadeDistance, Param4]}
  n3 = Add(A=TextureCoordinate() {uTiling=8, vTiling=8}, B=AppendVector(A=@n2.OffsetUVs, B=Add(A=@n2.OffsetUVs)))
  n4 = Floor(None=@n2.NoiseScale)
  n5 = Multiply(
   A=ParticleSubUV!(UVs, Tex, Apply View MipBias) {texture=/Game/Effects/Fort_Effects/BluePrints/BeeHive/T_SmokeFlip_03_8x8}.A
   B=TextureSample(UVs=Multiply(A=@n3, B=@n4) {desc=Scale}) {texture=&T_Inky_Smoke_Tile}.A
  )
  n6 = ScalarParameter() {defaultValue=256, parameterName=Camera Fade Distance}
  n7 = DynamicParameter() {paramNames [OffsetUVs, NoiseScale, MaxFadeDistance, Param4]}
  DepthFade(
   Opacity=Multiply(
    A=Multiply(A=@n1.A, B=@n5)
    B=MaterialFunctionCall(Distance=@n6) {materialFunction=&Near_Camera_Fade}.Result
   )
   FadeDistance=LinearInterpolate(
    A=Multiply(A=@n7.MaxFadeDistance) {constB=0.5}
    B=@n7.MaxFadeDistance
    Alpha=ParticleSubUV() {texture=/Game/Effects/Fort_Effects/Textures/Smoke/T_SmokeFlip_03_8x8_Depth}.G
   ) {constA=32}
  )
  MaterialFunctionCall(
   BaseNormal=MaterialFunctionCall(
    Normal=ParticleSubUV() {samplerType=SAMPLERTYPE_Normal, texture=/Game/Effects/Fort_Effects/Textures/Smoke/T_SmokeFlip_03_8x8_N}.RGB
    Flatness=DynamicParameter() {paramNames [OffsetUVs, NoiseScale, MaxFadeDistance, Param4]}.Param4
   ) {materialFunction=&FlattenNormal}.Result
   AdditionalNormal=TextureSample(UVs=TextureCoordinate() {uTiling=8, vTiling=8}) {samplerType=SAMPLERTYPE_Normal, texture=&T_Sphere_N}.RGB
  ) {materialFunction=&BlendAngleCorrectedNormals}
  Multiply(A=VectorParameter() {parameterName=SmokeColor, defaultValue=<rgba 0.5 0.5 0.5 0>}.RGB, B=@n1.RGB)
 ]
}
