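# M_DebugMat is a translucent, Default Lit material in the Athena prototype Cube folder, named as a debug material.
# Its graph holds only a Vertex Color node and one scalar parameter with a low default.
# No links from either node to material outputs are shown, so how they drive color or opacity is not stated.
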Material /Game/Athena/Prototype/Blueprints/Cube/M_DebugMat
graph {
 properties {blendMode=BLEND_Translucent, pixelDepthOffsetMode=PDOM_Legacy, usedShadingModels=Default Lit}
 nodes [
  ScalarParameter() {defaultValue=0.2}
  VertexColor()
 ]
}
