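# Two-sided lit material for the prototype Cube's HLOD proxy, imitating the cube's look with procedural gradients.
# A centered gradient from the U axis of UV channel 2 is squared, then separately shaped for base color and emissive.
# Base color blends a noise-modulated tint into a second color, while emissive blends two scaled colors, all packed through SetMaterialAttributes.
# The radial gradient and three-color emissive lerp nodes are left unconnected.
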Material /Game/Athena/Prototype/Blueprints/Cube/Materials/M_Cube_HLODProxy
graph {
 properties {
  bUseMaterialAttributes=true
  bUsedWithMeshDeformer=true
  bUsedWithMeshParticles=true
  bUsedWithSkeletalMesh=true
  pixelDepthOffsetMode=PDOM_Legacy
  twoSided=true
  usedShadingModels=Default Lit
 }
 nodes [
  MaterialFunctionCall(
   UVs=TextureCoordinate() {coordinateIndex=2}
   Radius=ScalarParameter() {defaultValue=0.6, parameterName=HLODRad}
   Density=ScalarParameter() {defaultValue=1, parameterName=HLODDens}
  ) {materialFunction=&RadialGradientExponential}
  n1 = VectorParameter() {parameterName=HLOD_Col0_Em, defaultValue=<rgba 0 0 0 1>}
  n2 = VectorParameter() {parameterName=HLOD_Col1_Em, defaultValue=<rgba 0 0.3763 1.2 1>}
  n3 = Multiply(A=Subtract(A=TextureCoordinate() {coordinateIndex=2}) {constB=0.5}) {constB=2}
  n4 = MaterialFunctionCall(Float2=OneMinus(None=Abs(None=@n3))) {materialFunction=&BreakOutFloat2Components}
  n5 = Multiply(A=@n4.R, B=@n4.R)
  n6 = ScalarParameter() {defaultValue=0.8, parameterName=powerEm}
  n7 = Saturate(None=Multiply(A=Power(Base=@n5, Exp=@n6), B=ScalarParameter() {defaultValue=1, parameterName=multGradEm}))
  MaterialFunctionCall(1=@n1.RGB, 2=@n2.RGB, 3=VectorParameter() {parameterName=HLOD_Col2_Em, defaultValue=<rgba 0.6875 1.4126 3 1>}.RGB, A=@n7) {materialFunction=&Lerp_Multiple_Float3}
  n8 = VectorParameter() {parameterName=HLOD_Col0, defaultValue=<rgba 0.1655 0.5827 0.5885 1>}
  n9 = TextureSample() {samplerType=SAMPLERTYPE_Masks, texture=/Game/Global/Textures/Default/Noise/T_CubeNoise}
  n10 = Multiply(
   A=Power(Base=@n5, Exp=ScalarParameter() {defaultValue=0.8, parameterName=power})
   B=ScalarParameter() {defaultValue=5, parameterName=multGrad}
  )
  SetMaterialAttributes!(
   MaterialAttributes
   Base Color=LinearInterpolate(
    A=Multiply(A=@n8.RGB, B=@n9.R)
    B=VectorParameter() {parameterName=HLOD_Col1, defaultValue=<rgba 0.0235 0.0831 0.2188 1>}.RGB
    Alpha=Saturate(None=@n10)
   )
   Metallic=ScalarParameter() {defaultValue=0.2, parameterName=Metal}
   Specular=ScalarParameter() {defaultValue=0.06, parameterName=Specular}
   Roughness=ScalarParameter() {defaultValue=0.8, parameterName=Rough}
   Emissive Color=Multiply(A=LinearInterpolate(A=@n1.RGB, B=@n2.RGB, Alpha=@n7), B=ScalarParameter() {defaultValue=1, parameterName=EmissiveMult})
  ) {attributeSetTypes [
    #1
    #2
    #3
    #4
    #5
   ]}
 ]
}
