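# Purple Cube material: a two-sided lit surface with parallax clouds, sparkles, darkened edges and glowing runes.
# Emissive fakes a pulsing inner sphere by ray-sphere math, tinting cloud noise through four colors and adding layered bump-offset runes.
# RuneActive, Zap and Zap2 drive glow strength; CubeJiggleIntensity blends in rotational vertex wobble.
# Switches add sinking-gradient and Cubetown lake variants, with cheaper low-quality paths and a simplified material-proxy output.
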Material /Game/Athena/Prototype/Blueprints/Cube/Materials/M_Cube_Purple
graph {
 properties {
  bUseMaterialAttributes=true
  bUsedWithInstancedStaticMeshes=true
  bUsedWithMeshDeformer=true
  bUsedWithMeshParticles=true
  bUsedWithNiagaraMeshParticles=true
  bUsedWithSkeletalMesh=true
  pixelDepthOffsetMode=PDOM_Legacy
  previewMesh=/Engine/Transient.MaterialEditorMeshComponent_11
  twoSided=true
  usedShadingModels=Default Lit
 }
 nodes [
  n1 = Multiply(A=TextureCoordinate(), B=ScalarParameter() {defaultValue=0.15, parameterName=CloudNoiseScale})
  n2 = Transform(None=CameraVectorWS()) {transformSourceType=TRANSFORMSOURCE_World, transformType=TRANSFORM_Tangent}
  n3 = ComponentMask(None=@n2) {g=true, r=true}
  Add(A=@n1, B=Multiply(A=@n3, B=ScalarParameter() {defaultValue=0.172, parameterName=Parallax}) {constB=0.1})
  Constant3Vector() {constant=<rgba 0 0 1 0>}
  Constant() {r=1}
  Constant() {r=1}
  n4 = ComponentMask(None=@n2) {b=true}
  Divide(A=@n3, B=@n4)
  n5 = VertexColor()
  n6 = VectorParameter() {parameterName=BaseDiffuse, defaultValue=<rgba 0.015 0.0013 0.1 1>}
  n7 = ScalarParameter() {defaultValue=1, parameterName=StarsTileMod}
  n8 = TextureCoordinate()
  n9 = Multiply(A=Constant() {r=14}, B=@n8)
  n10 = Constant() {r=-2}
  n11 = Constant() {r=0.02}
  n12 = TextureSample(UVs=Multiply(A=@n7, B=BumpOffset(Coordinate=@n9, Height=@n10, HeightRatioInput=@n11))) {samplerSource=SSM_Wrap_WorldGroupSettings, samplerType=SAMPLERTYPE_Masks, texture=/Game/Global/Textures/Default/Noise/T_CubeNoise}
  n13 = Multiply(A=@n8, B=Constant() {r=28})
  n14 = Constant() {r=0.2}
  n15 = TextureSample(UVs=Multiply(A=@n7, B=BumpOffset(Coordinate=@n13, Height=@n10, HeightRatioInput=@n14))) {samplerSource=SSM_Wrap_WorldGroupSettings, samplerType=SAMPLERTYPE_Masks, texture=&T_CubeNoise}
  n16 = Power(Base=Multiply(A=@n12.B, B=@n15.B), Exp=Constant() {r=7})
  n17 = Constant() {r=60}
  n18 = Multiply(
   A=VectorParameter() {parameterName=SparklesColor, defaultValue=<rgba 0.5857 0.2 1 1>}.RGB
   B=NamedRerouteDeclaration(None=Multiply(A=@n16, B=@n17)) {name=Sparkles, nodeColor=<rgba 0 0.5078 1 1>}
  )
  n19 = BumpOffset(Coordinate=@n1, Height=ScalarParameter() {defaultValue=0.58, parameterName=CloudParallax}, HeightRatioInput=Constant() {r=-0.5})
  n20 = TextureSample!(
   UVs=Panner(
    Coordinate=Rotator(Coordinate=@n19) {centerX=0.25, centerY=0.333, speed=0.0075}
    Speed=ScalarParameter() {defaultValue=0.004, parameterName=CloudSpeed}
   )
   Tex
   Level=ScalarParameter() {defaultValue=5, parameterName=mipbias1}
   Apply View MipBias
  ) {mipValueMode=TMVM_MipLevel, samplerType=SAMPLERTYPE_Masks, texture=&T_CubeNoise}
  n21 = Constant() {r=3}
  n22 = Saturate(None=Add(A=Power(Base=@n20.R, Exp=@n21), B=Constant() {r=-0.05}))
  n23 = TextureCoordinate() {coordinateIndex=2}
  n24 = Constant2Vector() {g=0.5, r=0.5}
  n25 = MaterialFunctionCall(Float2=Abs(None=Subtract(A=@n23, B=@n24))) {materialFunction=&BreakOutFloat2Components}
  n26 = Max(A=@n25.R, B=@n25.R)
  n27 = Saturate(None=Multiply(A=OneMinus(None=@n26), B=Constant() {r=1.7}))
  n28 = Constant() {r=30}
  n29 = Add(A=@n6.RGB, B=Multiply(A=Multiply(A=@n18, B=@n22), B=Power(Base=@n27, Exp=@n28)))
  n30 = Constant() {r=1.95}
  n31 = Power(Base=Saturate(None=Multiply(A=@n26, B=@n30)), Exp=Constant() {r=5})
  n32 = Power(Base=@n20.R, Exp=ScalarParameter() {defaultValue=3, parameterName=CloudPower})
  n33 = TwoSidedSign()
  n34 = Multiply(
   A=LinearInterpolate(
    A=QualitySwitch(Default=@n29, Low=@n6.RGB, High=@n29)
    B=VectorParameter() {parameterName=EdgeDiffuse, defaultValue=<rgba 0.006 0.0005 0.04 1>}.RGB
    Alpha=MaterialFunctionCall(Transition Phase=@n31, Height Texture=@n32) {materialFunction=&HeightLerp}.Lerp Alpha No Contrast
   )
   B=@n33
  )
  n35 = Fresnel() {baseReflectFraction=0}
  n36 = ScalarParameter() {defaultValue=1, parameterName=MetalExp}
  n37 = Multiply(A=Saturate(None=Add(A=@n35, B=@n36)), B=ScalarParameter() {defaultValue=0.2, parameterName=Metal})
  n38 = Constant() {r=0.2}
  n39 = Saturate(None=Multiply(A=@n26, B=Constant() {r=2}))
  n40 = Constant() {r=1.8}
  n41 = Add(
   A=MaterialFunctionCall(Transition Phase=Power(Base=@n39, Exp=@n40), Height Texture=@n32) {materialFunction=&HeightLerp}.Lerp Alpha No Contrast
   B=@n32
  )
  n42 = Multiply(A=@n33, B=LinearInterpolate(A=Constant() {r=0.2}, B=Constant() {r=0.1}, Alpha=Saturate(None=@n41)))
  n43 = Saturate(None=Multiply(A=@n26, B=Constant() {r=1.4}))
  n44 = Constant() {r=1.5}
  n45 = LinearInterpolate(
   A=ScalarParameter() {defaultValue=0.29, parameterName=Rough-0}
   B=ScalarParameter() {defaultValue=0.8, parameterName=Rough-1}
   Alpha=MaterialFunctionCall(Transition Phase=Power(Base=@n43, Exp=@n44), Height Texture=@n32) {materialFunction=&HeightLerp}.Lerp Alpha No Contrast
  )
  n46 = Multiply(A=NamedRerouteUsage(), B=@n45)
  n47 = VectorParameter() {parameterName=HLOD_Col0, defaultValue=<rgba 0 0 0 1>}
  n48 = VectorParameter() {parameterName=HLOD_Col1, defaultValue=<rgba 0.0746 0 0.1042 1>}
  n49 = Saturate(None=MaterialFunctionCall(
    UVs=TextureCoordinate() {coordinateIndex=2}
    Radius=ScalarParameter() {defaultValue=0.8, parameterName=HLODRad}
    Density=ScalarParameter() {defaultValue=1, parameterName=HLODDens}
   ) {materialFunction=&RadialGradientExponential}.RadialGradientExponential)
  n50 = Constant3Vector() {constant=<rgba 1 1 1 0>}
  n51 = ParticleColor()
  n52 = ScalarParameter() {parameterName=Use Particle Color}
  n53 = WorldPosition()
  n54 = ScalarParameter() {defaultValue=-300, group=SinkingGradient, parameterName=SinkGradOffset}
  n55 = Multiply(
   A=Multiply(A=Add(A=@n53.XYZ, B=@n54), B=Constant() {r=-1})
   B=ScalarParameter() {defaultValue=0.0009, group=SinkingGradient, parameterName=SinkGradSize}
  )
  n56 = Power(
   Base=ComponentMask(None=Saturate(None=@n55)) {b=true}
   Exp=ScalarParameter() {defaultValue=13, group=SinkingGradient, parameterName=SinkGradPower}
  )
  n57 = VectorParameter() {group=SinkingGradient, parameterName=SinkGradColor, defaultValue=<rgba 0.8957 0 1.25 1>}
  n58 = VectorParameter() {parameterName=GlowColor1, defaultValue=<rgba 0.012 0.0011 0.08 1>}
  n59 = VectorParameter() {parameterName=GlowColor2, defaultValue=<rgba 0.0356 0.001 0.2 1>}
  n60 = VectorParameter() {parameterName=GlowColor3, defaultValue=<rgba 0.1264 0 0.3 1>}
  n61 = VectorParameter() {parameterName=GlowColor4, defaultValue=<rgba 0.3583 0 0.5 1>}
  n62 = ObjectRadius()
  n63 = MaterialFunctionCall(Value=Multiply(A=Time(), B=ScalarParameter() {defaultValue=0.5, parameterName=PulseSpeed_3})) {materialFunction=&LinearSine}
  n64 = Multiply(
   A=LinearInterpolate(A=Multiply(A=@n62) {constB=0.8}, B=@n62, Alpha=@n63."Rounded Linear Sine ")
   B=ScalarParameter() {defaultValue=1, parameterName=RadiusMult}
  )
  n65 = Multiply(A=@n64, B=@n64) {desc="r^2"}
  n66 = ObjectPositionWS()
  n67 = Add(A=VectorParameter() {parameterName=offz, defaultValue=<rgba 0 0 -400 1>}.RGB, B=@n66)
  n68 = ScalarParameter() {parameterName=RuneActive}
  n69 = Subtract(
   A=CameraPositionWS()
   B=Add(A=VectorParameter() {desc=For zapping players, parameterName=SphereOffset}.RGB, B=LinearInterpolate(A=@n66, B=@n67, Alpha=@n68))
  )
  n70 = Distance(A=@n69, B=Constant())
  n71 = DotProduct(A=@n69, B=CameraVectorWS())
  n72 = Subtract(A=Multiply(A=@n70, B=@n70), B=Multiply(A=@n71, B=@n71))
  n73 = SquareRoot(None=Max(A=Subtract(A=@n65, B=@n72)) {constB=0})
  n74 = Saturate(None=Divide(A=Multiply(A=@n73) {constB=2}, B=Multiply(A=@n64) {constB=2}))
  n75 = Multiply(A=@n74, B=ScalarParameter() {defaultValue=0.87, parameterName=gr334})
  n76 = TextureCoordinate() {coordinateIndex=2}
  n77 = Constant() {r=0.5}
  n78 = ScalarParameter() {defaultValue=0.8, parameterName=InitialSphereMaskRadius}
  n79 = ScalarParameter() {parameterName=InitialSphereMaskHardness}
  n80 = Saturate(None=MaterialFunctionCall(
    Input=SphereMask(A=@n76, B=@n77, Radius=@n78, Hardness=@n79) {attenuationRadius=0.7, hardnessPercent=30}
    Target Low=ScalarParameter() {defaultValue=0.2, parameterName=MaskMin}
   ) {materialFunction=&RemapValueRange}.Result)
  n81 = Saturate(None=Add(A=@n32, B=ScalarParameter() {parameterName=4444235}))
  n82 = ScalarParameter() {defaultValue=2.75, parameterName=2222}
  n83 = ScalarParameter() {defaultValue=1.5, parameterName=434233}
  n84 = ScalarParameter() {defaultValue=3, parameterName=CloudPower}
  n85 = Multiply(A=@n81, B=Power(Base=Power(Base=@n20.G, Exp=@n84), Exp=LinearInterpolate(A=@n82, B=@n83, Alpha=@n63."Rounded Linear Sine ")))
  n86 = MaterialFunctionCall(Transition Phase=@n75, Height Texture=Multiply(A=@n80, B=QualitySwitch(Default=@n85, Low=@n81, High=@n85))) {materialFunction=&HeightLerp}
  n87 = Multiply(
   A=Power(Base=@n75, Exp=ScalarParameter() {defaultValue=20, parameterName=gpow3})
   B=ScalarParameter() {defaultValue=75, parameterName=ActiveHotspot}
  )
  n88 = Multiply(A=@n87, B=@n32)
  n89 = LinearInterpolate(
   B=LinearInterpolate(A=@n87, B=@n88, Alpha=ScalarParameter() {defaultValue=1, parameterName=Blend-Tex-Sphere})
   Alpha=ScalarParameter() {defaultValue=1, parameterName=SphereMaskInfluence}
  ) {constA=1}
  n90 = LinearInterpolate(A=Multiply(A=@n80, B=@n89), B=@n89, Alpha=ScalarParameter() {parameterName=RemoveEdgeFromInnerGlow})
  n91 = Multiply(
   A=Add(
    A=MaterialFunctionCall(1=@n58.RGB, 2=@n59.RGB, 3=@n60.RGB, 4=@n61.RGB, A=@n86.Lerp Alpha No Contrast) {materialFunction=&Lerp_Multiple_Float3}.Lerp 4 Inputs
    B=Multiply(A=@n61.RGB, B=@n90)
   )
   B=@n80
  )
  n92 = ScalarParameter() {defaultValue=0.25, parameterName=deactivated_1}
  n93 = ScalarParameter() {defaultValue=0.05, parameterName=activated_1}
  n94 = Saturate(None=Add(A=LinearInterpolate(A=@n92, B=@n93, Alpha=@n68), B=@n32))
  n95 = Multiply(A=@n80, B=@n86.Lerp Alpha No Contrast)
  n96 = ScalarParameter() {defaultValue=2.5, parameterName=gpow3_1}
  n97 = LinearInterpolate(
   A=ScalarParameter() {defaultValue=-0.1, parameterName=deactivated}
   B=ScalarParameter() {defaultValue=1, parameterName=activated}
   Alpha=@n68
  )
  n98 = MaterialFunctionCall(Float2=Multiply(A=ScalarParameter() {defaultValue=75, parameterName=BB_15}, B=TextureCoordinate() {uTiling=0.95, vTiling=0.75})) {materialFunction=&BreakOutFloat2Components}
  n99 = MaterialFunctionCall(X=Add(A=ScalarParameter() {parameterName=BB_19}, B=@n98.R), Y=Add(A=@n98.G, B=ScalarParameter() {defaultValue=0.5, parameterName=BB_20})) {materialFunction=&MakeFloat2}
  n100 = TextureSampleParameter2D(UVs=@n99.Result) {
   parameterName=RuneTex
   samplerSource=SSM_Wrap_WorldGroupSettings
   samplerType=SAMPLERTYPE_LinearColor
   texture=&/DFHeightMask
  }
  n101 = LinearInterpolate(A=Constant() {r=-1}, B=Constant() {r=1}, Alpha=@n100.G) {constA=-1}
  n102 = ScalarParameter() {defaultValue=1, parameterName=qw1_2}
  n103 = TextureSampleParameter2D(UVs=BumpOffset(Coordinate=@n99.Result, Height=Multiply(A=@n101, B=@n102), HeightRatioInput=ScalarParameter() {defaultValue=0.06, parameterName=BB_16})) {
   parameterName=RuneTex
   samplerSource=SSM_Wrap_WorldGroupSettings
   samplerType=SAMPLERTYPE_LinearColor
   texture=&DFHeightMask
  }
  n104 = Power(
   Base=@n103.G
   Exp=LinearInterpolate(
    A=ScalarParameter() {defaultValue=10, parameterName=G_8}
    B=ScalarParameter() {defaultValue=2, parameterName=G_7}
    Alpha=@n63."Rounded Linear Sine "
   )
  )
  n105 = MaterialFunctionCall(Float2=Multiply(A=ScalarParameter() {defaultValue=19, parameterName=BB_3}, B=TextureCoordinate() {vTiling=0.75})) {materialFunction=&BreakOutFloat2Components}
  n106 = MaterialFunctionCall(
   X=Add(A=ScalarParameter() {defaultValue=0.1, parameterName=BB_13}, B=@n105.R)
   Y=Add(A=@n105.G, B=ScalarParameter() {defaultValue=0.05, parameterName=BB_14})
  ) {materialFunction=&MakeFloat2}
  n107 = TextureSampleParameter2D(UVs=@n106.Result) {
   parameterName=RuneTex
   samplerSource=SSM_Wrap_WorldGroupSettings
   samplerType=SAMPLERTYPE_LinearColor
   texture=&DFHeightMask
  }
  n108 = LinearInterpolate(A=Constant() {r=-1}, B=Constant() {r=1}, Alpha=@n107.G) {constA=-1}
  n109 = ScalarParameter() {defaultValue=0.5, parameterName=qw1_1}
  n110 = TextureSampleParameter2D(UVs=BumpOffset(Coordinate=@n106.Result, Height=Multiply(A=@n108, B=@n109), HeightRatioInput=ScalarParameter() {defaultValue=0.15, parameterName=BB_8})) {
   parameterName=RuneTex
   samplerSource=SSM_Wrap_WorldGroupSettings
   samplerType=SAMPLERTYPE_LinearColor
   texture=&DFHeightMask
  }
  n111 = Power(Base=@n110.G, Exp=ScalarParameter() {defaultValue=20, parameterName=G_2})
  n112 = Saturate(None=Multiply(A=@n110.G, B=ScalarParameter() {defaultValue=1, parameterName=G_5}))
  n113 = Multiply(A=ScalarParameter() {defaultValue=32, parameterName=BB_2}, B=TextureCoordinate())
  n114 = TextureSampleParameter2D(UVs=@n113) {
   parameterName=RuneTex
   samplerSource=SSM_Wrap_WorldGroupSettings
   samplerType=SAMPLERTYPE_LinearColor
   texture=&DFHeightMask
  }
  n115 = BumpOffset(
   Coordinate=@n113
   Height=LinearInterpolate(A=ScalarParameter() {defaultValue=-1, parameterName=qw1_8}, B=ScalarParameter() {defaultValue=1, parameterName=qw1_9}, Alpha=@n114.R)
   HeightRatioInput=ScalarParameter() {defaultValue=0.06, parameterName=BB_4}
  )
  n116 = TextureSampleParameter2D(UVs=QualitySwitch(Default=@n115, Low=@n113, High=@n115)) {
   parameterName=RuneTex
   samplerSource=SSM_Wrap_WorldGroupSettings
   samplerType=SAMPLERTYPE_LinearColor
   texture=&DFHeightMask
  }
  n117 = ScalarParameter() {defaultValue=1.5, parameterName=G_6}
  n118 = ScalarParameter() {defaultValue=15, parameterName=G_1}
  n119 = Multiply(
   A=Power(Base=@n116.R, Exp=LinearInterpolate(A=@n117, B=@n118, Alpha=@n63."Rounded Linear Sine "))
   B=ScalarParameter() {defaultValue=5, parameterName=G_9}
  )
  n120 = Panner(
   Coordinate=Multiply(A=ScalarParameter() {defaultValue=2, parameterName=BB_1}, B=TextureCoordinate())
   Time=Multiply(A=ScalarParameter() {defaultValue=0.0135, parameterName=BB_7}, B=Time())
  ) {speedX=1}
  n121 = Saturate(None=LinearInterpolate(
    A=ScalarParameter() {defaultValue=1, parameterName=qw1_3}
    B=ScalarParameter() {defaultValue=-1, parameterName=qw1_4}
    Alpha=TextureSample!(UVs=@n120, Tex, Apply View MipBias) {samplerType=SAMPLERTYPE_LinearColor, texture=&DFHeightMask}.B
   ))
  n122 = Saturate(None=Multiply(A=Multiply(A=@n116.R, B=@n121), B=ScalarParameter() {defaultValue=1.25, parameterName=G_3}))
  n123 = Multiply(
   A=VectorParameter() {parameterName=GlowColor-Runes, defaultValue=<rgba 0.6076 0 1 1>}.RGB
   B=LinearInterpolate(A=LinearInterpolate(A=@n104, B=@n111, Alpha=@n112), B=@n119, Alpha=@n122)
  )
  n124 = Multiply(A=Max(A=Multiply(A=@n97, B=@n123)) {constB=0}, B=ScalarParameter() {defaultValue=1, parameterName=Rune Gain})
  n125 = Add(A=Multiply(A=@n91, B=@n94), B=Multiply(A=Power(Base=@n95, Exp=@n96), B=Multiply(A=@n94, B=@n124)))
  n126 = ScalarParameter() {parameterName=Zap}
  n127 = ScalarParameter() {parameterName=Zap2}
  n128 = Add(A=Add(A=Multiply(A=@n126) {constB=2}, B=Multiply(A=@n127) {constB=1.5}), B=@n68)
  n129 = LinearInterpolate(
   A=Multiply(A=ScalarParameter() {defaultValue=0.18, parameterName=InnerGlowWhileOff}, B=@n125)
   B=Multiply(A=@n125, B=Constant() {r=1})
   Alpha=@n128
  )
  n130 = ScalarParameter() {defaultValue=0.1, parameterName=LowQ-InnerGlowWhileOff}
  n131 = Multiply(A=Max(A=Multiply(A=@n119, B=@n122)) {constB=0}, B=ScalarParameter() {defaultValue=1, parameterName=Rune Gain})
  n132 = ScalarParameter() {defaultValue=2, parameterName=Rune Gain Low Extra}
  n133 = Multiply(A=@n91, B=@n32)
  n134 = Multiply(A=@n133, B=Multiply(A=Saturate(None=@n80), B=Multiply(A=@n131, B=@n132) {constB=2}))
  n135 = Add(A=@n133, B=@n134)
  n136 = ScalarParameter() {defaultValue=0.4, parameterName=LowQ-InnerGlowWhileOn}
  n137 = QualitySwitch(Default=@n129, Low=LinearInterpolate(A=Multiply(A=@n130, B=@n135), B=Multiply(A=@n135, B=@n136), Alpha=@n128), High=@n129)
  n138 = StaticSwitchParameter(True=Add(A=Multiply(A=@n56, B=@n57.RGB), B=@n137), False=@n137) {group=SinkingGradient, parameterName=UseSinkingGradient}
  n139 = Multiply(A=Multiply(A=@n138, B=LinearInterpolate(A=@n50, B=@n51.RGB, Alpha=@n52)), B=ScalarParameter() {defaultValue=1, parameterName=Glow Gain})
  n140 = Multiply(A=@n5.R, B=Multiply(A=@n33, B=@n139))
  n141 = Add(A=@n140, B=NamedRerouteUsage())
  n142 = Time()
  n143 = Frac(None=Multiply(A=@n142) {constB=0.0344})
  n144 = LinearInterpolate(A=Constant3Vector() {constant=<rgba 1 0 0 0>}, B=Constant3Vector() {constant=<rgba 0 1 0 0>}, Alpha=Round(None=@n143))
  n145 = ScalarParameter() {defaultValue=0.01, parameterName=Max Wiggle}
  n146 = Sine(None=Multiply(A=@n142) {constB=0.004})
  n147 = Multiply(A=Time(), B=LinearInterpolate(B=ScalarParameter() {defaultValue=1.5, parameterName=ScaleTime}, Alpha=Saturate(None=@n146)))
  n148 = Subtract(A=Multiply(A=Frac(None=@n147)) {constB=2})
  n149 = LinearInterpolate(A=Multiply(A=@n145) {constB=-1}, B=@n145, Alpha=OneMinus(None=Abs(None=@n148)))
  n150 = MaterialFunctionCall() {materialFunction=&ObjectPivotPoint}
  n151 = WorldPosition()
  n152 = SetMaterialAttributes!(
   MaterialAttributes
   Base Color=MaterialProxyReplace(Realtime=Multiply(A=@n5.R, B=@n34), MaterialProxy=Constant())
   Metallic=MaterialProxyReplace(Realtime=StaticSwitchParameter(True=@n37, False=@n38) {parameterName=UseAdjustableMetal}, MaterialProxy=Constant())
   Specular=MaterialProxyReplace(Realtime=Multiply(A=@n5.R, B=@n42), MaterialProxy=Constant() {r=0.5})
   Roughness=MaterialProxyReplace(Realtime=StaticSwitchParameter(True=@n46, False=@n45) {group=Cubetown, parameterName=CubetownLake}, MaterialProxy=Constant() {r=1})
   Emissive Color=MaterialProxyReplace(
    Realtime=StaticSwitchParameter(True=@n141, False=@n140) {group=Cubetown, parameterName=CubetownLake}
    MaterialProxy=LinearInterpolate(A=@n47.RGB, B=@n48.RGB, Alpha=@n49)
   )
   World Position Offset=LinearInterpolate(
    B=RotateAboutAxis(NormalizedRotationAxis=@n144, RotationAngle=@n149, PivotPoint=@n150.Object Pivot Location, Position=@n151.XYZ)
    Alpha=ScalarParameter() {parameterName=CubeJiggleIntensity}
   ) {constAlpha=1}
  ) {attributeSetTypes [
    #1
    #2
    #3
    #4
    #5
    #6
   ]}
  MaterialFunctionCall(MA In=@n152) {materialFunction=&MF_CreativeMoveTool}
  MaterialFunctionCall(
   Input=MaterialFunctionCall!(
    UVs=@n76
    Size=ScalarParameter() {defaultValue=0.3, parameterName=SDFCubeSize}
    CornerRoundness=ScalarParameter() {parameterName=SDFCornerRoundness}
    KeepAspectRatio
    PixelSize
   ) {materialFunction=/FortUILibrary/MaterialFunctions/Shapes/SDFs/MF_SDF_Box}.SDF
   Input Low=ScalarParameter() {defaultValue=0.4, parameterName=SDFCubeMin}
   Input High=Constant() {r=1}
  ) {materialFunction=&RemapValue_0-1}
  n153 = TextureSample() {samplerType=SAMPLERTYPE_Normal, texture=/Game/Athena/Apollo/Water/Textures/02_Rounded/02_WaterFoam_Veroni_base_N}
  MaterialFunctionCall!(
   Normal=MaterialFunctionCall(Normal=Multiply(A=@n153.RGB) {constB=-1}, Flatness=ScalarParameter() {defaultValue=0.85, parameterName=FlattenCloud}) {materialFunction=&FlattenNormal}.Result
   CameraVector=@n2
  ) {materialFunction=/Engine/Functions/Engine_MaterialFunctions02/Utility/CustomReflectionVector}
  MaterialFunctionCall(Input=@n74, Target Low=ScalarParameter() {parameterName=EdgeNormalizedMin}) {materialFunction=&RemapValueRange}
  Multiply(A=@n4)
  Multiply(A=@n80, B=NamedRerouteUsage())
  n154 = NamedRerouteUsage()
  n155 = VectorParameter() {group=Cubetown, parameterName=EdgeOf Town, defaultValue=<rgba 15 1.9391 5.9363 0>}
  n156 = Constant() {r=0.025}
  n157 = TextureSample!(UVs=BumpOffset(Height=Constant() {r=0.4}), Tex, Level=Constant() {r=1}, Apply View MipBias) {
   mipValueMode=TMVM_MipLevel
   samplerType=SAMPLERTYPE_Grayscale
   texture=/Game/Environments/Apollo/Sets/CubeTownV2/Materials/Parent/T_CT_LakeTownEdge_H
  }
  n158 = Multiply(A=@n157.R, B=Constant() {r=0.5})
  n159 = TextureSample(UVs=BumpOffset(Height=Constant() {r=0.2}), Level=Constant() {r=3}) {
   mipValueMode=TMVM_MipLevel
   samplerType=SAMPLERTYPE_Grayscale
   texture=&T_CT_LakeTownEdge_H
  }
  n160 = Multiply(A=@n159.R, B=Constant() {r=0.25})
  n161 = TextureSample!(UVs, Tex, Apply View MipBias) {samplerType=SAMPLERTYPE_Grayscale, texture=&T_CT_LakeTownEdge_H}
  n162 = Multiply(A=Add(A=Multiply(A=@n154, B=@n155.RGB), B=Multiply(A=@n155.RGB, B=@n156)), B=Add(A=Add(A=@n158, B=@n160), B=@n161.R))
  n163 = Multiply(A=Multiply(A=@n155.RGB, B=@n161.R), B=Constant() {r=0.1})
  NamedRerouteDeclaration(None=StaticSwitchParameter(True=NamedRerouteUsage(), False=Add(A=@n162, B=@n163)) {group=Cubetown, parameterName=CubetownPad}) {name=CubetownEdgeGlow, nodeColor=<rgba 0.0312 1 0 1>}
  n164 = DotProduct(A=VertexNormalWS(), B=CameraVectorWS())
  n165 = Constant() {r=0.8}
  n166 = Constant() {r=0.3}
  n167 = WorldPosition()
  n168 = Constant() {r=256}
  n169 = TextureSample(UVs=ComponentMask(None=Divide(A=@n167.XYZ, B=@n168)) {g=true, r=true}) {
   samplerSource=SSM_Wrap_WorldGroupSettings
   samplerType=SAMPLERTYPE_Masks
   texture=/Game/Environments/Apollo/Sets/CubeTownV2/Materials/Parent/T_CubeTownDirt_M
  }
  NamedRerouteDeclaration(None=Add(A=Saturate(None=@n164), B=LinearInterpolate(A=@n165, B=@n166, Alpha=@n169.B))) {name=Cubetown Rough, nodeColor=<rgba 0 1 0.0391 1>}
  n170 = Multiply(
   A=TextureSample(UVs=TextureCoordinate() {coordinateIndex=2}) {samplerType=SAMPLERTYPE_Grayscale, texture=/Game/Environments/Apollo/Sets/CubeTownV2/Materials/Parent/T_Edges}.R
   B=VectorParameter() {group=Cubetown, parameterName=EdgeOf Town, defaultValue=<rgba 15 1.9391 5.9363 0>}.RGB
  )
  n171 = Constant() {r=0.5}
  n172 = Multiply(A=@n170, B=NamedRerouteUsage())
  n173 = Constant() {r=15}
  NamedRerouteDeclaration(None=Add(A=Multiply(A=@n170, B=@n171), B=Multiply(A=@n172, B=@n173))) {name=CubetownPad, nodeColor=<rgba 0 1 0.3203 1>}
  NamedRerouteDeclaration(None=Power(Base=@n80)) {name=SDFCubeMask, nodeColor=<rgba 1 0 0.3984 1>}
  NamedRerouteDeclaration(None=@n134) {name=Cloud, nodeColor=<rgba 1 0.6797 0 1>}
  QualitySwitch(Default=@n139, Low=@n138)
  QualitySwitch(Default=@n88, Low=@n87, High=@n88)
  ScalarParameter() {defaultValue=1400, parameterName=UpperRadius}
  ScalarParameter() {defaultValue=1200, parameterName=LowerRadius}
  SetMaterialAttributes!(MaterialAttributes, Emissive Color=NamedRerouteUsage()) {attributeSetTypes [#5]}
  SmoothStep(Min=@n143)
  TextureObject() {samplerType=SAMPLERTYPE_Masks, texture=&T_CubeNoise}
 ]
}
