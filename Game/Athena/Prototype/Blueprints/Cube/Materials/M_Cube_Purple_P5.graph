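# Two-sided lit material for the prototype purple cube, showing an empty glowing shell filling with animated "Nectar".
# The MPC_Cube Nectar value sets a UV2 fill level, which panning noise and HeightLerp turn into a soft boundary between empty and filled looks.
# Edge masks, crack textures and an IsCrack switch shape emissive, while a camera-ray sphere thickness calculation fakes an inner ball glow.
# MPC Heartbeat, Spin and SpinStartup values boost the glows.
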Material /Game/Athena/Prototype/Blueprints/Cube/Materials/M_Cube_Purple_P5
graph {
 properties {pixelDepthOffsetMode=PDOM_Legacy, twoSided=true, usedShadingModels=Default Lit}
 nodes [
  n1 = ObjectPositionWS()
  Add(
   A=Add(A=@n1, B=VectorParameter() {parameterName=offz, defaultValue=<rgba 0 0 200 1>}.RGB)
   B=VectorParameter() {desc=For zapping players, parameterName=SphereOffset}.RGB
  )
  Constant3Vector() {constant=<rgba 0 0 -400 1>}
  Constant() {r=1}
  Constant() {r=0.7}
  Constant() {r=0.1}
  Constant() {r=0.2}
  Constant() {r=8}
  Constant()
  Constant() {r=0.2}
  Constant() {r=1}
  Constant() {r=0.2}
  Constant() {r=0.6}
  Constant() {r=0.8}
  Constant() {r=0.29}
  n2 = ScalarParameter() {defaultValue=2, parameterName=DiffuseCornerMaskMod}
  n3 = SphereMask(
   A=TextureCoordinate() {coordinateIndex=1}
   B=Constant() {r=0.5}
   Radius=ScalarParameter() {defaultValue=0.7, parameterName=Crk-02_2}
   Hardness=ScalarParameter() {defaultValue=0.6, parameterName=Crk-02_1}
  ) {attenuationRadius=0.7, hardnessPercent=30}
  n4 = TextureCoordinate() {coordinateIndex=1}
  n5 = Constant2Vector() {g=0.5, r=0.5}
  n6 = MaterialFunctionCall(Float2=Abs(None=Subtract(A=@n4, B=@n5))) {materialFunction=&BreakOutFloat2Components}
  n7 = Max(A=@n6.R, B=@n6.R)
  n8 = OneMinus(None=@n7)
  n9 = Power(Base=OneMinus(None=Multiply(A=@n3, B=@n8)), Exp=ScalarParameter() {defaultValue=2.5, parameterName=Crk-02})
  n10 = Multiply(
   A=Saturate(None=Multiply(A=@n9, B=@n2))
   B=VectorParameter() {parameterName=EmptyContainerEdges-Diffuse, defaultValue=<rgba 0.046 0.0224 0.2 1>}.RGB
  )
  n11 = VectorParameter() {parameterName=EmptyContainer-Diffuse, defaultValue=<rgba 0.0044 0 0.04 1>}
  n12 = ScalarParameter() {defaultValue=1.5, parameterName=Nectar-Softness}
  n13 = Multiply(A=@n12, B=Constant() {r=-1})
  n14 = TextureCoordinate() {coordinateIndex=2}
  n15 = CollectionParameter() {collection=&/MPC_Cube, parameterName=Nectar}
  n16 = LinearInterpolate(
   A=ScalarParameter() {defaultValue=0.3, parameterName=NectarLevelLow}
   B=ScalarParameter() {defaultValue=-0.75, parameterName=NectarLevelHigh}
   Alpha=@n15
  )
  n17 = Saturate(None=Add(A=ComponentMask(None=@n14) {r=true}, B=@n16))
  n18 = ScalarParameter() {defaultValue=3, parameterName=dasf34t_10}
  n19 = TextureCoordinate() {coordinateIndex=1}
  n20 = Time()
  n21 = ScalarParameter() {defaultValue=0.075, parameterName=dasf34t_7}
  n22 = TextureSample(UVs=Panner(Coordinate=Multiply(A=@n18, B=@n19), Time=Multiply(A=@n20, B=@n21)) {speedX=-1, speedY=0.5}) {samplerType=SAMPLERTYPE_Grayscale, texture=&LowResBlurredNoise}
  n23 = Multiply(A=ScalarParameter() {defaultValue=2, parameterName=dasf34t_12}, B=TextureCoordinate() {coordinateIndex=1})
  n24 = Multiply(A=Time(), B=ScalarParameter() {defaultValue=0.1, parameterName=dasf34t_11})
  n25 = Multiply(
   A=@n22.R
   B=TextureSample(UVs=Panner(Coordinate=@n23, Time=@n24) {speedX=1, speedY=-0.5}) {samplerType=SAMPLERTYPE_Grayscale, texture=&LowResBlurredNoise}.R
  )
  n26 = MaterialFunctionCall(
   Transition Phase=Saturate(None=LinearInterpolate(A=@n12, B=@n13, Alpha=@n17))
   Height Texture=@n25
   Contrast=ScalarParameter() {defaultValue=10, parameterName=Nectar-Softness-B}
  ) {materialFunction=&HeightLerp}
  n27 = Multiply(
   A=Power(Base=@n26.Lerp Alpha No Contrast, Exp=ScalarParameter() {defaultValue=3, parameterName=444h43_1})
   B=ScalarParameter() {defaultValue=3, parameterName=444h43}
  )
  n28 = Saturate(None=Add(A=@n26.Alpha, B=Saturate(None=@n27)))
  n29 = OneMinus(None=@n28)
  Reroute(None=Multiply(A=Add(A=@n10, B=@n11.RGB), B=@n29))
  Reroute(None=ScalarParameter() {defaultValue=0.2, parameterName=Metal})
  n30 = ScalarParameter() {defaultValue=0.2, parameterName=SpecAmt_1}
  n31 = ScalarParameter() {defaultValue=0.1, parameterName=SpecAmt_2}
  n32 = TextureSample(UVs=Multiply(A=Constant() {r=3.5}, B=TextureCoordinate())) {samplerType=SAMPLERTYPE_Masks, texture=/Game/Global/Textures/Default/Noise/T_CubeNoise}
  Reroute(None=Multiply(A=ScalarParameter() {defaultValue=1, parameterName=SpecAmt}, B=LinearInterpolate(A=@n30, B=@n31, Alpha=@n32.R)))
  n33 = LinearInterpolate(
   A=ScalarParameter() {defaultValue=0.29, parameterName=RoughAmt_3}
   B=ScalarParameter() {defaultValue=0.8, parameterName=RoughAmt_4}
   Alpha=@n32.R
  )
  Reroute(None=@n33)
  n34 = VectorParameter() {parameterName=BallGlow, defaultValue=<rgba 0.7801 0.224 1 1>}
  n35 = ScalarParameter() {defaultValue=20, parameterName=KevinSmall-Power}
  n36 = CollectionParameter() {collection=&MPC_Cube, parameterName=Spin}
  n37 = LinearInterpolate(
   A=ScalarParameter() {defaultValue=900, parameterName=SphereRadius}
   B=ScalarParameter() {defaultValue=1200, parameterName=SphereRadiusHot}
   Alpha=@n36
  )
  n38 = Subtract(A=CameraPositionWS(), B=@n1)
  n39 = Distance(A=@n38, B=Constant())
  n40 = Multiply(A=@n39, B=@n39)
  n41 = DotProduct(A=@n38, B=CameraVectorWS())
  n42 = Multiply(A=@n41, B=@n41)
  n43 = Max(A=Subtract(A=Multiply(A=@n37, B=@n37) {desc="r^2"}, B=Subtract(A=@n40, B=@n42))) {constB=0}
  n44 = Divide(A=Multiply(A=SquareRoot(None=@n43)) {constB=2}, B=Multiply(A=@n37) {constB=2})
  n45 = VertexColor()
  n46 = Multiply(A=@n45.B, B=Multiply(A=Power(Base=@n44, Exp=@n35), B=ScalarParameter() {defaultValue=0.02, parameterName=KevinSmall-Mod}))
  n47 = Power(Base=@n44, Exp=ScalarParameter() {defaultValue=2, parameterName=KevinLarge-Power})
  n48 = Multiply(A=ScalarParameter() {defaultValue=0.02, parameterName=KevinLarge-Mod}, B=@n47)
  n49 = ScalarParameter() {defaultValue=1, parameterName=dasf34t_3}
  n50 = ScalarParameter() {defaultValue=4, parameterName=CornerGlow-Heartbeat_2}
  n51 = CollectionParameter() {collection=&MPC_Cube, parameterName=Heartbeat}
  n52 = LinearInterpolate(A=ScalarParameter() {parameterName=dasf34t_4}, B=ScalarParameter() {defaultValue=300, parameterName=Spin-Bright}, Alpha=@n36)
  n53 = Multiply(A=Saturate(None=Add(A=@n46, B=@n48)), B=Add(A=LinearInterpolate(A=@n49, B=@n50, Alpha=@n51), B=@n52))
  n54 = Constant()
  n55 = ScalarParameter() {defaultValue=40, parameterName=SpinStart}
  n56 = CollectionParameter() {collection=&MPC_Cube, parameterName=SpinStartup}
  n57 = Add(A=@n53, B=Multiply(A=@n53, B=LinearInterpolate(A=@n54, B=@n55, Alpha=@n56)))
  n58 = Multiply(A=@n29, B=VectorParameter() {parameterName=EmptyOverallGlow, defaultValue=<rgba 0 0 0 1>}.RGB)
  n59 = ScalarParameter() {defaultValue=1.3153, parameterName=CrackTile_1}
  n60 = Rotator(
   Coordinate=Multiply(A=ScalarParameter() {defaultValue=4, parameterName=CrackTile_2}, B=TextureCoordinate())
   Time=ScalarParameter() {defaultValue=2.9926, parameterName=CrackTile_5}
  )
  n61 = Multiply(A=TextureSample(UVs=Add(A=@n59, B=@n60)) {texture=/Game/Effects/Fort_Effects/Textures/Tile/T_Glass_Cracks_M}.G, B=@n9)
  n62 = LinearInterpolate(
   A=ScalarParameter() {defaultValue=-3.85, parameterName=CornerGlow-LerpA_1}
   B=ScalarParameter() {defaultValue=4.2, parameterName=CornerGlow-LerpB_1}
   Alpha=@n7
  )
  n63 = Add(
   A=Multiply(A=@n9, B=ScalarParameter() {defaultValue=0.25, parameterName=CornerGlow-Dull-Mod})
   B=Multiply(A=Saturate(None=@n62), B=ScalarParameter() {defaultValue=1, parameterName=CornerGlow-Sharp-Mod})
  )
  n64 = Add(A=@n61, B=@n63)
  n65 = Add(
   A=Multiply(A=VectorParameter() {parameterName=EmptyEdgeGlow, defaultValue=<rgba 0.7449 0.1 1 1>}.RGB, B=@n64)
   B=Multiply(A=VectorParameter() {parameterName=EmptyCrackGlow, defaultValue=<rgba 0.1134 0.0229 0.2 1>}.RGB, B=@n45.B)
  )
  n66 = ScalarParameter() {defaultValue=0.04, parameterName=Spin-EdgeGlow}
  n67 = Add(A=Multiply(A=@n61, B=ScalarParameter() {defaultValue=3, parameterName=Crk-03_1}), B=@n64)
  n68 = Add(A=LinearInterpolate(A=Add(A=@n58, B=@n65), B=Multiply(A=@n34.RGB, B=@n57), Alpha=@n47), B=Multiply(A=Multiply(A=@n66, B=@n67), B=@n52))
  n69 = VectorParameter() {parameterName=CrackEmissiveColor-Low, defaultValue=<rgba 0.2 0.01 0.1962 1>}
  n70 = VectorParameter() {parameterName=CrackEmissiveColor-High, defaultValue=<rgba 20 1 19.6247 1>}
  n71 = LinearInterpolate(
   A=LinearInterpolate(
    A=ScalarParameter() {defaultValue=2, parameterName=dasf34t_20}
    B=ScalarParameter() {defaultValue=5, parameterName=CornerGlow-Heartbeat_1}
    Alpha=@n51
   )
   B=ScalarParameter() {defaultValue=-0.1, parameterName=dasf34t_21}
   Alpha=@n45.A
  )
  n72 = TextureCoordinate() {coordinateIndex=2}
  n73 = ScalarParameter() {defaultValue=0.82, parameterName=NectarLevel-eh}
  n74 = Saturate(None=Add(A=ComponentMask(None=@n72) {r=true}, B=Add(A=@n16, B=@n73)))
  n75 = ScalarParameter() {defaultValue=15, parameterName=NectarLevelBubble-Empty}
  n76 = ScalarParameter() {defaultValue=3, parameterName=NectarLevelBubble-Full}
  n77 = Multiply(A=@n26.Alpha, B=Power(Base=@n74, Exp=LinearInterpolate(A=@n75, B=@n76, Alpha=@n15)))
  n78 = Multiply(A=ScalarParameter() {defaultValue=8, parameterName=dasf34t_2}, B=TextureCoordinate() {coordinateIndex=1})
  n79 = Multiply(A=Time(), B=ScalarParameter() {defaultValue=0.4, parameterName=dasf34t_1})
  n80 = ScalarParameter() {defaultValue=1.5, parameterName=dasf34t_9}
  n81 = TextureSample(UVs=Add(A=Panner(Coordinate=@n78, Time=@n79) {speedX=-1, speedY=0.5}, B=Multiply(A=@n80, B=@n25))) {samplerType=SAMPLERTYPE_Grayscale, texture=&LowResBlurredNoise}
  n82 = ScalarParameter() {defaultValue=3, parameterName=dasf34t_8}
  n83 = Multiply(A=Multiply(A=Power(Base=@n81.R, Exp=@n82), B=ScalarParameter() {defaultValue=12, parameterName=dasf34t_6}), B=@n22.R)
  n84 = Saturate(None=Add(A=Multiply(A=@n77, B=@n83), B=@n25))
  n85 = Saturate(None=Add(A=Saturate(None=@n71), B=Multiply(A=@n45.B, B=@n84)))
  n86 = ScalarParameter() {defaultValue=0.6, parameterName=CornerGlow-MASTER}
  n87 = ScalarParameter() {defaultValue=2, parameterName=CornerGlow-Heartbeat}
  n88 = ScalarParameter() {defaultValue=20, parameterName=dasf34t_17}
  n89 = Multiply(
   A=Multiply(
    A=VectorParameter() {parameterName=SurfaceEmissive-C, defaultValue=<rgba 0.0804 0 0.2 1>}.RGB
    B=LinearInterpolate(A=@n86, B=@n87, Alpha=@n51)
   )
   B=Add(A=Multiply(A=@n84, B=@n88), B=ScalarParameter() {defaultValue=0.2, parameterName=dasf34t_18})
  )
  n90 = ScalarParameter() {defaultValue=40, parameterName=Crk-03}
  n91 = Multiply(A=Add(A=Multiply(A=@n90, B=@n61), B=@n63), B=VectorParameter() {parameterName=SurfaceEmissive-A, defaultValue=<rgba 0.7166 0 1 1>}.RGB)
  n92 = LinearInterpolate(
   A=ScalarParameter() {defaultValue=20, parameterName=NectarLevelGrad-Empty}
   B=ScalarParameter() {defaultValue=5, parameterName=NectarLevelGrad-Full}
   Alpha=@n15
  )
  n93 = Add(
   A=Multiply(A=Power(Base=@n74, Exp=@n92), B=VectorParameter() {parameterName=GlowColor4_2, defaultValue=<rgba 1.25 0.125 1.1941 1>}.RGB)
   B=StaticSwitchParameter(True=LinearInterpolate(A=@n69.RGB, B=@n70.RGB, Alpha=@n85), False=Add(A=@n91, B=@n89)) {parameterName=IsCrack}
  )
  n94 = Saturate(None=Add(A=@n32.R, B=ScalarParameter() {defaultValue=0.5, parameterName=SurfaceTextureAmount}))
  Reroute(None=LinearInterpolate(A=Multiply(A=@n33, B=@n68), B=Multiply(A=@n93, B=@n94), Alpha=@n28))
 ]
}
