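# Masked prototype material for the Cube that draws a scorched, purple-swirled ground patch shaped by an offsettable DFHeightMask texture.
# That mask drives height lerps blending world-projected grass through outer and inner tints into rock-textured ash.
# It also gates two purple color layers built from counter-panning storm-swirl samples, with UVs distorted by mask or rock alpha.
# A vertical WPO offset follows the mask, with optional sphere mask and vertex-red blending.
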Material /Game/Athena/Prototype/Blueprints/Cube/Materials/M_DFHeight
graph {
 properties {blendMode=BLEND_Masked, ditherOpacityMask=true, pixelDepthOffsetMode=PDOM_Legacy, usedShadingModels=Default Lit}
 nodes [
  n1 = VectorParameter() {parameterName=Color1_2, defaultValue=<rgba 0.202 0 1 1>}
  n2 = VectorParameter() {parameterName=Color1_9, defaultValue=<rgba 0.7526 0 1 1>}
  n3 = ScalarParameter() {parameterName=U-Offset}
  n4 = MaterialFunctionCall(Float2=Multiply(A=TextureCoordinate(), B=ScalarParameter() {defaultValue=0.5, parameterName=SymbolSize})) {materialFunction=&BreakOutFloat2Components}
  n5 = ScalarParameter() {parameterName=V-Offset}
  n6 = TextureSample(UVs=MaterialFunctionCall(X=Add(A=@n3, B=@n4.R), Y=Add(A=@n4.G, B=@n5)) {materialFunction=&MakeFloat2}.Result) {samplerType=SAMPLERTYPE_LinearColor, texture=&/DFHeightMask}
  n7 = StaticSwitchParameter(True=@n6.R, False=@n6.R) {defaultValue=true, parameterName=RT-GF}
  n8 = LinearInterpolate(A=ScalarParameter() {defaultValue=-3, parameterName=G_53}, B=ScalarParameter() {defaultValue=0.8, parameterName=G_58}, Alpha=@n7)
  n9 = ScalarParameter() {defaultValue=0.5, parameterName=G_11}
  n10 = Constant() {r=-0.5}
  n11 = AppendVector(A=Add(A=@n10), B=Add(A=@n10, B=OneMinus(None=@n7)))
  n12 = Multiply(A=TextureCoordinate(), B=ScalarParameter() {defaultValue=1.5, parameterName=SymbolSize_8})
  n13 = Multiply(A=ScalarParameter() {defaultValue=-0.04, parameterName=hhhh_6}, B=Time())
  n14 = TextureSample(UVs=Add(A=Multiply(A=@n9, B=@n11), B=Panner(Coordinate=@n12, Time=@n13) {speedX=0.5, speedY=-0.5})) {samplerType=SAMPLERTYPE_LinearGrayscale, texture=/Game/Global/Textures/Default/Flowmaps/T_StormSwirlMask}
  n15 = Multiply(A=TextureCoordinate(), B=ScalarParameter() {defaultValue=2, parameterName=SymbolSize_7})
  n16 = Multiply(A=ScalarParameter() {defaultValue=0.035, parameterName=hhhh_5}, B=Time())
  n17 = Multiply(A=@n11, B=Constant() {r=-1})
  n18 = ScalarParameter() {defaultValue=0.5, parameterName=G_3}
  n19 = TextureSample(UVs=Add(A=Panner(Coordinate=@n15, Time=@n16) {speedX=0.5, speedY=-0.5}, B=Multiply(A=@n17, B=@n18))) {samplerType=SAMPLERTYPE_LinearGrayscale, texture=&T_StormSwirlMask}
  n20 = TextureSample(UVs=Multiply(A=TextureCoordinate(), B=ScalarParameter() {defaultValue=2, parameterName=GravelSize_1})) {texture=/Game/Packages/DS_Fortnight/WhiteRockCustom_D}
  n21 = OneMinus(None=@n20.A)
  n22 = Power(Base=@n21, Exp=Constant() {r=3})
  n23 = Constant() {r=0.2}
  n24 = Multiply(
   A=MaterialFunctionCall(Transition Phase=Clamp(None=@n8), Height Texture=Multiply(A=@n14.R, B=@n19.R)) {materialFunction=&HeightLerp}.Lerp Alpha No Contrast
   B=Saturate(None=Add(A=@n22, B=@n23))
  )
  n25 = VectorParameter() {parameterName=Color1_13, defaultValue=<rgba 0.3564 0 1 1>}
  n26 = VectorParameter() {parameterName=Color1_14, defaultValue=<rgba 0.7526 0 1 1>}
  n27 = ScalarParameter() {defaultValue=-4, parameterName=G_54}
  n28 = ScalarParameter() {defaultValue=1, parameterName=G_55}
  n29 = Multiply(A=TextureCoordinate(), B=ScalarParameter() {defaultValue=2, parameterName=SymbolSize_2})
  n30 = Multiply(A=ScalarParameter() {defaultValue=-0.04, parameterName=hhhh_4}, B=Time())
  n31 = Constant() {r=-0.5}
  n32 = LinearInterpolate(B=AppendVector(A=Add(A=@n31, B=@n20.A), B=Add(A=@n31, B=@n21)), Alpha=ScalarParameter() {defaultValue=0.03, parameterName=G_49})
  n33 = TextureSample(UVs=Add(A=Panner(Coordinate=@n29, Time=@n30) {speedX=0.5, speedY=-0.5}, B=@n32)) {samplerType=SAMPLERTYPE_LinearGrayscale, texture=&T_StormSwirlMask}
  n34 = Multiply(A=TextureCoordinate(), B=ScalarParameter() {defaultValue=5, parameterName=SymbolSize_1})
  n35 = Multiply(A=ScalarParameter() {defaultValue=0.02, parameterName=hhhh_3}, B=Time())
  n36 = TextureSample(UVs=Add(A=@n32, B=Panner(Coordinate=@n34, Time=@n35) {speedX=0.5, speedY=-0.5})) {samplerType=SAMPLERTYPE_LinearGrayscale, texture=&T_StormSwirlMask}
  n37 = Multiply(A=Saturate(None=LinearInterpolate(A=@n27, B=@n28, Alpha=@n7)), B=Multiply(A=Multiply(A=@n33.R, B=@n36.R), B=@n20.A))
  Add(
   A=Multiply(A=LinearInterpolate(A=@n1.RGB, B=@n2.RGB, Alpha=@n24), B=@n24)
   B=Multiply(A=LinearInterpolate(A=@n25.RGB, B=@n26.RGB, Alpha=@n37), B=@n37)
  )
  Constant() {r=0.5}
  Constant() {r=0.5}
  n38 = VectorParameter() {parameterName=MainGroundColorMod, defaultValue=<rgba 1 1 1 1>}
  n39 = WorldPosition() {worldPositionShaderOffset=WPT_ExcludeAllShaderOffsets}
  n40 = TextureSampleParameter2D(UVs=Divide(A=ComponentMask(None=@n39.XYZ) {g=true, r=true}, B=ScalarParameter() {defaultValue=700, parameterName=MainGroundTextureScale})) {parameterName=MainGroundTexture, texture=/Game/Athena/Environments/Landscape/Textures/T_Athena_Grass_Farm_ColorMatched_D_2}
  n41 = VectorParameter() {parameterName=OuterColorA, defaultValue=<rgba 0.02 0.0195 0.0159 1>}
  n42 = Desaturation(None=@n40.RGB, Fraction=ScalarParameter() {parameterName=GroundColorDesat})
  n43 = Multiply(A=@n41.RGB, B=@n42)
  n44 = Saturate(None=LinearInterpolate(
    A=ScalarParameter() {defaultValue=-0.2, parameterName=GroundColorMaskA}
    B=ScalarParameter() {defaultValue=2, parameterName=GroundColorMaskB}
    Alpha=@n40.G
   ))
  n45 = LinearInterpolate(A=@n41.RGB, B=VectorParameter() {parameterName=OuterColorB, defaultValue=<rgba 0.3333 0.4 0.2833 1>}.RGB, Alpha=@n44)
  n46 = LinearInterpolate(A=Constant(), B=Constant() {r=2}, Alpha=@n7)
  n47 = VectorParameter() {parameterName=InnerColorA, defaultValue=<rgba 0.0036 0.0039 0.005 1>}
  n48 = VectorParameter() {parameterName=InnerColorB, defaultValue=<rgba 0.24 0.2613 0.3 1>}
  n49 = LinearInterpolate(A=Constant() {r=-1.5}, B=Constant() {r=2}, Alpha=@n7)
  n50 = OneMinus(None=@n40.A)
  n51 = LinearInterpolate(
   A=LinearInterpolate(
    A=Multiply(A=@n40.RGB, B=@n38.RGB)
    B=StaticSwitchParameter(True=@n43, False=@n45) {parameterName=ModGroundDiffuse}
    Alpha=Saturate(None=@n46)
   )
   B=StaticSwitchParameter(True=Multiply(A=@n47.RGB, B=@n42), False=LinearInterpolate(A=@n47.RGB, B=@n48.RGB, Alpha=@n44)) {parameterName=ModGroundDiffuse}
   Alpha=MaterialFunctionCall(Transition Phase=Saturate(None=@n49), Height Texture=@n50) {materialFunction=&HeightLerp}.Lerp Alpha No Contrast
  )
  n52 = ScalarParameter() {defaultValue=0.02, parameterName=AshDarkness}
  n53 = Saturate(None=LinearInterpolate(A=ScalarParameter() {defaultValue=0.3, parameterName=AshMaskA}, B=ScalarParameter() {defaultValue=1.2, parameterName=AshMaskB}, Alpha=@n20.A))
  n54 = Saturate(None=LinearInterpolate(A=ScalarParameter() {defaultValue=-2.2, parameterName=AshSizeA}, B=ScalarParameter() {defaultValue=1.5, parameterName=AshSizeB}, Alpha=@n7))
  n55 = ScalarParameter() {defaultValue=1, parameterName=AshSizeContrast}
  n56 = LinearInterpolate(A=Constant() {r=-5}, B=Constant() {r=2}, Alpha=@n7)
  n57 = OneMinus(None=MaterialFunctionCall(Transition Phase=Saturate(None=@n56), Height Texture=@n20.A) {materialFunction=&HeightLerp}.Lerp Alpha No Contrast)
  n58 = Constant() {r=0.5}
  Multiply(
   A=LinearInterpolate(
    A=Saturate(None=@n51)
    B=Multiply(A=@n52, B=@n53)
    Alpha=MaterialFunctionCall(Transition Phase=@n54, Height Texture=@n50, Contrast=@n55) {materialFunction=&HeightLerp}.Alpha
   )
   B=Saturate(None=Add(A=@n57, B=@n58))
  )
  n59 = LinearInterpolate(A=ScalarParameter() {parameterName=G_20}, B=ScalarParameter() {defaultValue=3, parameterName=G_21}, Alpha=@n7)
  Reroute(None=MaterialFunctionCall(Transition Phase=Clamp(None=@n59), Height Texture=@n50) {materialFunction=&HeightLerp}.Lerp Alpha No Contrast)
  n60 = Constant3Vector() {constant=<rgba 0 0 1 0>}
  n61 = ScalarParameter() {defaultValue=5, group=WPO, parameterName=LODMeshOffset}
  n62 = Saturate(None=LinearInterpolate(A=Constant() {r=-1}, B=Constant() {r=2}, Alpha=@n7))
  n63 = SphereMask(
   A=TextureCoordinate()
   B=MaterialFunctionCall(
    X=ScalarParameter() {defaultValue=0.5, group=WPO, parameterName=RadiusOffsetA}
    Y=ScalarParameter() {defaultValue=0.4, group=WPO, parameterName=RadiusOffsetB}
   ) {materialFunction=&MakeFloat2}.Result
   Radius=ScalarParameter() {defaultValue=0.4, group=WPO, parameterName=RadiusOffsetMask}
   Hardness=ScalarParameter() {defaultValue=0.5, group=WPO, parameterName=RadiusOffsetMaskHard}
  )
  n64 = ScalarParameter() {defaultValue=1, group=WPO, parameterName=RadiusOffsetMaskAmt}
  n65 = StaticSwitchParameter(True=Add(A=@n62, B=Multiply(A=@n63, B=@n64)), False=@n62) {group=WPO, parameterName=UseWPO-SphereMask}
  n66 = Constant3Vector() {constant=<rgba 0 0 1 0>}
  n67 = Multiply(
   A=LinearInterpolate(A=Constant(), B=ScalarParameter() {defaultValue=10, group=WPO, parameterName=VertOffset}, Alpha=Saturate(None=@n65))
   B=@n66
  )
  n68 = Multiply(A=@n66, B=ScalarParameter() {group=WPO, parameterName=MeshOffset})
  Reroute(None=LinearInterpolate(A=Multiply(A=@n60, B=@n61), B=Add(A=@n67, B=@n68), Alpha=VertexColor().R))
  ScalarParameter() {parameterName=spec}
  ScalarParameter() {defaultValue=0.7, parameterName=rough}
 ]
}
