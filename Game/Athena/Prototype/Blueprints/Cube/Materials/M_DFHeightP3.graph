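# Masked, lit material for the Cube prototype that renders world-projected rocky corrupted ground with an optional rune layer.
# An offsettable DFHeightMask drives height lerps against ground and white-rock textures, shaping darkened color, specular and masking.
# A distorted, panning storm-swirl texture adds purple glow inside the mask.
# Vertex red blends World Position Offset between a distance-faded LOD lift and mask-driven height, while Forward and Mobile paths simplify roughness and normals.
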Material /Game/Athena/Prototype/Blueprints/Cube/Materials/M_DFHeightP3
graph {
 properties {blendMode=BLEND_Masked, ditherOpacityMask=true, opacityMaskClipValue=0.45, pixelDepthOffsetMode=PDOM_Legacy, usedShadingModels=Default Lit}
 nodes [
  Constant() {r=5}
  Constant() {r=0.5}
  Constant() {r=0.5}
  Constant() {r=-90000}
  Constant() {r=-90000}
  n1 = VectorParameter() {parameterName=ccc_1, defaultValue=<rgba 0.005 0.005 0.005 1>}
  n2 = TextureCoordinate()
  n3 = Constant() {r=1}
  n4 = TextureSample(UVs=Add(A=Multiply(A=@n2, B=@n3), B=Constant() {r=0.633})) {texture=/Game/Packages/DS_Fortnight/WhiteRockCustom_D}
  LinearInterpolate(
   A=@n1.RGB
   B=VectorParameter() {parameterName=ccc_2, defaultValue=<rgba 0.05 0.05 0.05 1>}.RGB
   Alpha=Power(Base=OneMinus(None=@n4.R), Exp=Constant() {r=15})
  )
  n5 = ScalarParameter() {defaultValue=1536, parameterName=MainGroundTextureScale_1}
  MaterialFunctionCall(TextureObject=TextureObject() {samplerType=SAMPLERTYPE_Normal, texture=/Game/Maps/Environments/T_Corruptground_Rocky_N}, TextureSize=@n5) {materialFunction=&WorldAlignedNormal}
  n6 = ScalarParameter() {parameterName=U-Offset}
  n7 = MaterialFunctionCall(Float2=Multiply(A=TextureCoordinate(), B=ScalarParameter() {defaultValue=0.5, parameterName=SymbolSize})) {materialFunction=&BreakOutFloat2Components}
  n8 = ScalarParameter() {parameterName=V-Offset}
  n9 = TextureSample(UVs=MaterialFunctionCall(X=Add(A=@n6, B=@n7.R), Y=Add(A=@n7.G, B=@n8)) {materialFunction=&MakeFloat2}.Result) {samplerType=SAMPLERTYPE_LinearColor, texture=&/DFHeightMask}
  n10 = StaticSwitchParameter(True=@n9.R, False=@n9.R) {defaultValue=true, parameterName=RT-GF}
  n11 = LinearInterpolate(A=ScalarParameter() {parameterName=G_20}, B=ScalarParameter() {defaultValue=2, parameterName=G_21}, Alpha=@n10)
  n12 = Divide(A=ComponentMask(None=WorldPosition() {worldPositionShaderOffset=WPT_ExcludeAllShaderOffsets}.XYZ) {g=true, r=true}, B=@n5)
  n13 = TextureSampleParameter2D(UVs=@n12) {parameterName=MainGroundTexture_2, texture=/Game/Maps/Environments/T_Corruptground_Rocky2_D}
  n14 = OneMinus(None=@n13.A)
  Reroute(None=MaterialFunctionCall(Transition Phase=Clamp(None=@n11), Height Texture=@n14) {materialFunction=&HeightLerp}.Lerp Alpha No Contrast)
  n15 = Saturate(None=LinearInterpolate(A=Constant() {r=-1}, B=Constant() {r=2}, Alpha=@n10))
  n16 = ScalarParameter() {defaultValue=0.5, group=WPO, parameterName=RadiusOffsetA}
  n17 = ScalarParameter() {defaultValue=0.4, group=WPO, parameterName=RadiusOffsetB}
  n18 = Multiply(
   A=SphereMask(
    A=TextureCoordinate()
    B=MaterialFunctionCall(X=@n16, Y=@n17) {materialFunction=&MakeFloat2}.Result
    Radius=ScalarParameter() {defaultValue=0.4, group=WPO, parameterName=RadiusOffsetMask}
    Hardness=ScalarParameter() {defaultValue=0.5, group=WPO, parameterName=RadiusOffsetMaskHard}
   )
   B=ScalarParameter() {defaultValue=1, group=WPO, parameterName=RadiusOffsetMaskAmt}
  )
  Saturate(None=StaticSwitchParameter(True=Add(A=@n15, B=@n18), False=@n15) {group=WPO, parameterName=UseWPO-SphereMask})
  n19 = Constant() {r=1}
  ShadingPathSwitch(
   Default=LinearInterpolate(
    A=ScalarParameter() {defaultValue=0.77, parameterName=rough_2}
    B=ScalarParameter() {defaultValue=0.4, parameterName=rough_1}
    Alpha=TextureSample(UVs=@n12) {samplerType=SAMPLERTYPE_Masks, texture=/Game/Maps/Environments/T_Corruptground_Rocky2_S}.B
   )
   Forward=@n19
   Mobile=@n19
  )
  n20 = VertexNormalWS()
  n21 = WorldPosition()
  n22 = ScalarParameter() {defaultValue=1536, group=Corruption, parameterName=Corruption_TexScale}
  n23 = TextureSampleParameter2D(UVs=Divide(A=ComponentMask(None=@n21.XYZ) {g=true, r=true}, B=Multiply(A=@n22))) {group=Corruption, parameterName=Corruption_Normal, samplerType=SAMPLERTYPE_Normal, texture=&T_Corruptground_Rocky_N}
  n24 = ScalarParameter() {defaultValue=1, parameterName=NormalXY}
  n25 = ScalarParameter() {defaultValue=1, parameterName=NormalZ}
  ShadingPathSwitch(
   Default=Multiply(
    A=Transform(None=@n23.RGB) {transformSourceType=TRANSFORMSOURCE_World, transformType=TRANSFORM_Tangent}
    B=MaterialFunctionCall(X=@n24, Y=@n24, Z=@n25) {materialFunction=&MakeFloat3}.Result
   )
   Forward=@n20
   Mobile=@n20
  )
  n26 = Multiply(A=Desaturation(None=@n13.RGB, Fraction=Constant() {r=1}), B=ScalarParameter() {defaultValue=0.4, parameterName=34terge})
  n27 = MaterialFunctionCall(Transition Phase=Power(Base=@n10, Exp=Constant() {r=2.5}), Height Texture=Power(Base=@n14, Exp=Constant() {r=1.5})) {materialFunction=&HeightLerp}
  n28 = Constant() {r=2}
  n29 = TextureCoordinate()
  n30 = Constant() {r=1.5}
  n31 = LinearInterpolate(A=TextureSample(UVs=Multiply(A=@n29, B=@n30)) {texture=&WhiteRockCustom_D}.A, B=@n4.A)
  n32 = MaterialFunctionCall(Transition Phase=Power(Base=@n10, Exp=Constant() {r=8}), Height Texture=Saturate(None=Multiply(A=@n28, B=@n31)), Contrast=Constant() {r=0.5}) {materialFunction=&HeightLerp}
  n33 = LinearInterpolate(A=Constant() {r=-1}, B=Constant() {r=3}, Alpha=@n32.Alpha)
  StaticSwitchParameter(
   True=LinearInterpolate(A=LinearInterpolate(A=@n13.RGB, B=@n26, Alpha=@n27.Lerp Alpha No Contrast), B=@n1.RGB, Alpha=Saturate(None=@n33))
   False=@n13.RGB
  ) {defaultValue=true, parameterName=UseRunes}
  n34 = OneMinus(None=@n27.Lerp Alpha No Contrast)
  n35 = LinearInterpolate(
   A=ScalarParameter() {defaultValue=1, group=WPO, parameterName=R_A}
   B=ScalarParameter() {defaultValue=-1, group=WPO, parameterName=R_B}
   Alpha=@n32.Alpha
  )
  n36 = Multiply(A=@n13.R, B=ScalarParameter() {defaultValue=0.9, group=Corruption, parameterName=Corruption_SpecGain})
  StaticSwitchParameter(True=Multiply(A=Multiply(A=@n34, B=@n35), B=@n36), False=@n36) {defaultValue=true, parameterName=UseRunes}
  n37 = VectorParameter() {parameterName=Color1_13, defaultValue=<rgba 0.0891 0 0.25 1>}
  n38 = VectorParameter() {parameterName=Color1_14, defaultValue=<rgba 0.1881 0 0.25 1>}
  n39 = ScalarParameter() {defaultValue=-4, parameterName=G_54}
  n40 = ScalarParameter() {defaultValue=1, parameterName=G_55}
  n41 = TextureCoordinate()
  n42 = ScalarParameter() {defaultValue=0.5, parameterName=SymbolSize_2}
  n43 = ScalarParameter() {defaultValue=-0.04, parameterName=hhhh_4}
  n44 = Time()
  n45 = Constant() {r=-0.5}
  n46 = Add(A=@n45, B=@n4.A)
  n47 = Add(A=@n45, B=OneMinus(None=@n4.A))
  n48 = Add(
   A=Panner(Coordinate=Multiply(A=@n41, B=@n42), Time=Multiply(A=@n43, B=@n44)) {speedX=0.5, speedY=-0.5}
   B=LinearInterpolate(B=AppendVector(A=@n46, B=@n47), Alpha=ScalarParameter() {defaultValue=0.03, parameterName=G_49})
  )
  n49 = Multiply(
   A=Saturate(None=LinearInterpolate(A=@n39, B=@n40, Alpha=@n10))
   B=Multiply(
    A=TextureSample(UVs=@n48) {samplerType=SAMPLERTYPE_LinearGrayscale, texture=/Game/Global/Textures/Default/Flowmaps/T_StormSwirlMask}.R
    B=@n4.A
   )
  )
  n50 = Constant()
  StaticSwitchParameter(True=Multiply(A=LinearInterpolate(A=@n37.RGB, B=@n38.RGB, Alpha=@n49), B=@n49), False=@n50) {defaultValue=true, parameterName=UseRunes}
  n51 = MaterialFunctionCall!(
   Blend Range=ScalarParameter() {defaultValue=200000, parameterName=Fade-BlendRange}
   Start Offset=ScalarParameter() {defaultValue=-10000, parameterName=Fade-BlendRange_1}
  ) {materialFunction=/Engine/Functions/Engine_MaterialFunctions03/Particles/Distance_Blend}
  n52 = Multiply(A=Constant3Vector() {constant=<rgba 0 0 1 0>}, B=ScalarParameter() {defaultValue=150, group=WPO, parameterName=LODMeshOffset})
  n53 = Constant3Vector() {constant=<rgba 0 0 1 0>}
  n54 = Multiply(A=LinearInterpolate(A=Constant(), B=ScalarParameter() {defaultValue=10, group=WPO, parameterName=VertOffset}, Alpha=@n15), B=@n53)
  n55 = Multiply(A=@n53, B=ScalarParameter() {group=WPO, parameterName=MeshOffset})
  StaticSwitchParameter(True=LinearInterpolate(A=Multiply(A=@n51.Result, B=@n52), B=Add(A=@n54, B=@n55), Alpha=VertexColor().R), False=@n50) {defaultValue=true, parameterName=UseRunes}
 ]
}
